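# Translucent unlit Niagara mesh-particle material for the InstantGravelNoble skin's chest swoosh, with parameters naming it as lightning.
# Particle color feeds an emissive adjustment function, while cross products of view direction and a transformed Y axis widen the mesh across UV.x to face the camera.
# Particle alpha combines with width and length edge fades, two-layer panning anisotropic noise, an interpolated Fresnel falloff and a dynamic-parameter-driven reveal mask.
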Material /BRCosmetics/Characters/Player/Male/Medium/Bodies/M_MED_InstantGravelNoble/FX/Materials/M_Skin_InstantGravelNoble_ChestSwoosh
graph {
 properties {
  bUsedWithNiagaraMeshParticles=true
  blendMode=BLEND_Translucent
  pixelDepthOffsetMode=PDOM_Legacy
  shadingModel=MSM_Unlit
  usedShadingModels=Unlit
 }
 nodes [
  n1 = ParticleColor()
  MaterialFunctionCall(Emissive=@n1.RGB) {materialFunction=&MF_FX_EmissiveAdjustments}
  n2 = Normalize(VectorInput=Subtract(A=CameraPositionWS(), B=WorldPosition().XYZ))
  n3 = Normalize(VectorInput=Transform(None=Constant3Vector() {constant=<rgba 0 1 0 0>}))
  n4 = Normalize(VectorInput=CrossProduct(A=@n2, B=@n3))
  n5 = ComponentMask(None=TextureCoordinate()) {r=true}
  Multiply(A=@n4, B=Multiply(A=Subtract(A=@n5) {constB=0.5}, B=ScalarParameter() {defaultValue=7, parameterName=Lightning Width}))
  Multiply(A=MaterialFunctionCall() {materialFunction=&CameraDirectionVector}.Result) {constB=-1}
  n6 = Multiply(A=Multiply(A=NamedRerouteUsage(), B=NamedRerouteUsage()), B=NamedRerouteUsage())
  n7 = Power(Base=Abs(None=NamedRerouteUsage()), Exp=ScalarParameter() {defaultValue=3, parameterName=Fresnel Exponent})
  n8 = NamedRerouteUsage()
  Multiply(A=@n1.A, B=Multiply(A=Multiply(A=@n6, B=@n7), B=LinearInterpolate(Alpha=@n8) {constA=0.1}))
  n9 = TextureCoordinate()
  n10 = Abs(None=ConstantBiasScale(None=ComponentMask(None=@n9) {r=true}) {bias=-0.5, scale=2})
  n11 = Divide(
   A=Subtract(A=OneMinus(None=@n10), B=ScalarParameter() {parameterName=Width Fade Offset})
   B=ScalarParameter() {defaultValue=1, parameterName=Width Fade Size}
  )
  NamedRerouteDeclaration(None=Power(Base=Saturate(None=@n11), Exp=ScalarParameter() {defaultValue=3, parameterName=Width Fade Exponent})) {name=Width Fade, nodeColor=<rgba 0.8516 1 0 1>}
  n12 = TextureCoordinate()
  n13 = Abs(None=ConstantBiasScale(None=ComponentMask(None=@n12) {g=true}) {bias=-0.5, scale=2})
  n14 = Divide(
   A=Subtract(A=OneMinus(None=@n13), B=ScalarParameter() {defaultValue=0.05, parameterName=Length Fade Offset})
   B=ScalarParameter() {defaultValue=0.5, parameterName=Length Fade Size}
  )
  NamedRerouteDeclaration(None=Power(Base=Saturate(None=@n14), Exp=ScalarParameter() {defaultValue=3, parameterName=Length Fade Exponent})) {name=Length Fade, nodeColor=<rgba 0.6875 0 1 1>}
  n15 = TextureCoordinate()
  n16 = AppendVector(A=ComponentMask(None=@n15) {g=true}, B=ComponentMask(None=@n15) {r=true})
  n17 = VectorParameter() {parameterName=Noise UV Scale and Speed, defaultValue=<rgba 2.62 0.1 1 -0.01>}
  n18 = Multiply(A=@n16, B=ComponentMask(None=@n17.RGB) {g=true, r=true})
  n19 = AppendVector(A=@n17.B, B=@n17.B)
  n20 = VectorParameter() {parameterName=Noise Layer 2 UV Scale and Speed, defaultValue=<rgba 0.62 0.02 1.3 0.02>}
  n21 = Multiply(A=@n16, B=ComponentMask(None=@n20.RGB) {g=true, r=true})
  n22 = AppendVector(A=@n20.B, B=@n20.B)
  n23 = Add(
   A=TextureSample(UVs=Panner(Coordinate=@n18, Speed=@n19)) {samplerType=SAMPLERTYPE_LinearGrayscale, texture=/Game/Animation/Game/MainPlayer/Emotes/Congestion/FX/T_Noise_Anisotropic}.R
   B=TextureSample(UVs=Panner(Coordinate=@n21, Speed=@n22)) {samplerType=SAMPLERTYPE_LinearGrayscale, texture=&T_Noise_Anisotropic}.R
  )
  NamedRerouteDeclaration(None=Power(Base=Divide(A=@n23), Exp=ScalarParameter() {defaultValue=1, parameterName=Noise Exponent})) {name=Panning Noise, nodeColor=<rgba 1 0.3984 0 1>}
  n24 = ComponentMask(None=TextureCoordinate()) {g=true}
  n25 = DynamicParameter() {paramNames [Lightning Reveal Position, Lightning Reveal Size, Param3, Param4]}
  NamedRerouteDeclaration(None=MaterialFunctionCall!(Gradient=OneMinus(None=@n24), Mask Position=@n25.Lightning Reveal Position, Mask Size=@n25.Lightning Reveal Position) {materialFunction=/Game/Effects/Fort_Effects/MaterialFunctions/MF_GradientMask}.Sinusoidal) {name=Reveal Mask, nodeColor=<rgba 0 1 1 1>}
  n26 = Normalize(VectorInput=CrossProduct(A=@n3, B=@n4))
  NamedRerouteDeclaration(None=VertexInterpolator(VS=DotProduct(A=@n26, B=@n2)).PS) {name=Fresnel, nodeColor=<rgba 1 0 0.2812 1>}
 ]
}
