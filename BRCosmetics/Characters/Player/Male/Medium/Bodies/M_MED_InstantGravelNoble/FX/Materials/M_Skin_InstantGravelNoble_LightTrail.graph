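# Additive, unlit, two-sided Niagara ribbon material for the light trail effect on the InstantGravelNoble player skin.
# Two three-color gradients across V are blended by the VariantPicker dynamic parameter, tinted by particle color and passed through MF_FX_EmissiveAdjustments.
# A fade mask combines U length, V edge, particle alpha, view-angle and depth fades, skipping depth on mobile and low/medium quality.
# Dynamic parameters also control depth fade and a camera-directed offset.
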Material /BRCosmetics/Characters/Player/Male/Medium/Bodies/M_MED_InstantGravelNoble/FX/Materials/M_Skin_InstantGravelNoble_LightTrail
graph {
 properties {
  bUsedWithNiagaraRibbons=true
  blendMode=BLEND_Additive
  pixelDepthOffsetMode=PDOM_Legacy
  shadingModel=MSM_Unlit
  twoSided=true
  usedShadingModels=Unlit
 }
 nodes [
  n1 = VectorParameter() {parameterName=Color A Cyan, defaultValue=<rgba 0.3968 0.0844 0.4397 0.2>}
  n2 = VectorParameter() {parameterName=Color C Cyan, defaultValue=<rgba 0 0.8333 1 0.2>}
  n3 = TextureCoordinate()
  n4 = ComponentMask(None=@n3) {g=true}
  n5 = Subtract(A=@n4, B=@n1.A)
  n6 = MaterialFunctionCall(
   1=@n1.RGB
   2=VectorParameter() {parameterName=Color B Cyan, defaultValue=<rgba 0.2898 0.2 1 1>}.RGB
   3=@n2.RGB
   A=Saturate(None=Divide(A=@n5, B=@n2.A))
  ) {materialFunction=&Lerp_Multiple_Float3}
  n7 = VectorParameter() {parameterName=Color A, defaultValue=<rgba 0.245 0.05 0.5 0.2>}
  n8 = VectorParameter() {parameterName=Color C, defaultValue=<rgba 1 0.8333 0 0.2>}
  n9 = Subtract(A=@n4, B=@n7.A)
  n10 = MaterialFunctionCall(
   1=@n7.RGB
   2=VectorParameter() {parameterName=Color B, defaultValue=<rgba 0.4 0.16 0.2453 1>}.RGB
   3=@n8.RGB
   A=Saturate(None=Divide(A=@n9, B=@n8.A))
  ) {materialFunction=&Lerp_Multiple_Float3}
  n11 = DynamicParameter() {defaultValue=<rgba 0 1 1 0>, paramNames [VariantPicker, Depth Fade Offset, Depth Fade Distance, Camera Offset]}
  n12 = ParticleColor()
  MaterialFunctionCall(Emissive=Multiply(A=LinearInterpolate(A=@n6.Lerp 3 Inputs, B=@n10.Lerp 3 Inputs, Alpha=@n11.VariantPicker), B=@n12.RGB)) {materialFunction=&MF_FX_EmissiveAdjustments}
  n13 = ComponentMask(None=@n3) {r=true}
  n14 = ConstantBiasScale(None=@n4) {bias=-0.5, scale=2}
  n15 = Divide(A=OneMinus(None=Abs(None=@n14)), B=ScalarParameter() {defaultValue=0.2, parameterName=Edge Fade Size})
  n16 = Multiply(A=Power(Base=Saturate(None=@n15), Exp=ScalarParameter() {defaultValue=3, parameterName=Edge Fade Exponent}), B=@n12.A)
  n17 = Abs(None=DotProduct(A=VertexNormalWS(), B=CameraVectorWS()))
  n18 = ScalarParameter() {defaultValue=0.1, parameterName=Fresnel Fade Offset}
  n19 = Saturate(None=Divide(A=Subtract(A=@n17, B=@n18), B=ScalarParameter() {defaultValue=2, parameterName=Fresnel Fade Size}))
  n20 = DynamicParameter() {defaultValue=<rgba 0 1 1 0>, paramNames [VariantPicker, Depth Fade Offset, Depth Fade Distance, Camera Offset]}
  n21 = Subtract(A=Subtract(A=SceneDepth().R, B=PixelDepth().R), B=@n20.Depth Fade Offset)
  n22 = Constant() {r=1}
  n23 = ShadingPathSwitch(Default=Saturate(None=Divide(A=@n21, B=@n20.Depth Fade Distance)), Mobile=@n22) {desc=Migrated from FeatureLevelSwitch}
  Multiply(
   A=Power(Base=OneMinus(None=@n13), Exp=ScalarParameter() {defaultValue=3, parameterName=Length Fade Exponent})
   B=Multiply(A=Multiply(A=@n16, B=@n19), B=QualitySwitch(Default=@n23, Low=@n22, Medium=@n22))
  ) {constA=1}
  n24 = CameraPositionWS()
  n25 = WorldPosition()
  Multiply(A=Normalize(VectorInput=Subtract(A=@n24, B=@n25.XYZ)), B=@n20.Camera Offset)
 ]
}
