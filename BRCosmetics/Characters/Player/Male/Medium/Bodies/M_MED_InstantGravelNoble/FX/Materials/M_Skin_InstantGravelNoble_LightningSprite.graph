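# Translucent, unlit Niagara material that draws lightning zaps for the InstantGravelNoble character skin's effects, on sprites, ribbons and mesh particles.
# Dynamic parameters pan and scale U, shift colors and pick a cyan or yellow palette, graded by the zap texture and V gradient.
# UVs flip randomly per particle and scale about center to taper at the V ends.
# Texture-driven opacity uses near-camera fade and depth fade, the latter skipped on mobile and lower quality.
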Material /BRCosmetics/Characters/Player/Male/Medium/Bodies/M_MED_InstantGravelNoble/FX/Materials/M_Skin_InstantGravelNoble_LightningSprite
graph {
 properties {
  bUsedWithNiagaraMeshParticles=true
  bUsedWithNiagaraRibbons=true
  bUsedWithNiagaraSprites=true
  blendMode=BLEND_Translucent
  pixelDepthOffsetMode=PDOM_Legacy
  shadingModel=MSM_Unlit
  twoSided=true
  usedShadingModels=Unlit
 }
 nodes [
  n1 = DynamicParameter() {defaultValue=<rgba 0 1 1 0>, paramNames [Swipe, Scale U, Color Transition, VariantPicker]}
  n2 = LinearInterpolate(
   A=VectorParameter() {group=02 - Color, parameterName=CyanDarkColor, defaultValue=<rgba 0.0968 0.0886 0.3698 1>}.RGB
   B=VectorParameter() {group=02 - Color, parameterName=YellowDarkColor, defaultValue=<rgba 0.25 0.156 0 1>}.RGB
   Alpha=@n1.VariantPicker
  )
  n3 = ComponentMask(None=TextureCoordinate()) {g=true}
  n4 = LinearInterpolate(
   A=LinearInterpolate(
    A=VectorParameter() {group=02 - Color, parameterName=CyanMidDarkColor, defaultValue=<rgba 0.1133 0.0586 0.4167 1>}.RGB
    B=VectorParameter() {group=02 - Color, parameterName=YellowMidDarkColor, defaultValue=<rgba 0.3 0.1882 0 1>}.RGB
    Alpha=@n1.VariantPicker
   )
   B=@n2
   Alpha=SmoothStep(Value=@n3) {constMax=0.5, constMin=0.8}
  )
  n5 = LinearInterpolate(
   A=VectorParameter() {group=02 - Color, parameterName=CyanMidColor, defaultValue=<rgba 0.047 0.1212 0.5312 1>}.RGB
   B=VectorParameter() {group=02 - Color, parameterName=YellowMidColor, defaultValue=<rgba 0.4 0.3644 0.0541 1>}.RGB
   Alpha=@n1.VariantPicker
  )
  n6 = DynamicParameter() {defaultValue=<rgba 0 1 1 0>, paramNames [Swipe, Scale U, Color Transition, VariantPicker]}
  n7 = LinearInterpolate(A=@n5, B=@n2, Alpha=@n6.Color Transition)
  n8 = LinearInterpolate(
   A=VectorParameter() {group=02 - Color, parameterName=CyanLightColor, defaultValue=<rgba 0 1.5 1.2423 1>}.RGB
   B=VectorParameter() {group=02 - Color, parameterName=YellowLightColor, defaultValue=<rgba 0.8 0.793 0.3627 1>}.RGB
   Alpha=@n1.VariantPicker
  )
  n9 = TextureObjectParameter() {
   parameterName=Lightning Parameter
   texture=/BRCosmetics/Characters/Player/Male/Medium/Bodies/M_MED_InstantGravel/FX/InstantGravelNoble/T_InstantGravelNoble_FlatZapTex
  }
  n10 = TextureSample(UVs=NamedRerouteUsage(), Tex=@n9)
  n11 = ScalarParameter() {defaultValue=0.5, group=02 - Color, parameterName=ColorTransitionPoint}
  n12 = ScalarParameter() {defaultValue=0.3, group=02 - Color, parameterName=ColorTransitionSoftness}
  n13 = LinearInterpolate(
   A=@n7
   B=LinearInterpolate(A=@n7, B=LinearInterpolate(A=@n8, B=@n5, Alpha=@n6.Color Transition), Alpha=SmoothStep(Value=@n10.R) {constMax=0.4, constMin=0.1})
   Alpha=SmoothStep(Min=@n11, Max=Add(A=@n11, B=@n12) {constB=0.2}, Value=@n3) {constMax=0.7, constMin=0.5}
  )
  n14 = SmoothStep(Value=@n10.R) {constMax=0.1}
  n15 = ParticleColor()
  MaterialFunctionCall(Emissive=Multiply(A=@n15.RGB, B=LinearInterpolate(A=@n4, B=@n13, Alpha=@n14))) {materialFunction=&MF_FX_EmissiveAdjustments}
  n16 = MaterialFunctionCall(Distance=ScalarParameter() {defaultValue=250, group=03 - Opacity, parameterName=CameraDistance_Fade}) {materialFunction=&Near_Camera_Fade}
  n17 = Saturate(None=Add(A=SmoothStep(Value=@n10.R) {constMax=0.2}, B=Multiply(A=@n10.G) {constB=0.5}))
  n18 = DepthFade(FadeDistance=ScalarParameter() {defaultValue=11, group=03 - Opacity, parameterName=DepthFadeDist})
  n19 = Constant() {r=1}
  n20 = ShadingPathSwitch(Default=@n18, Deferred=@n18, Mobile=@n19) {desc=Migrated from FeatureLevelSwitch}
  Multiply(
   A=Saturate(None=Multiply(A=@n16.Result) {constB=5})
   B=Multiply(A=Multiply(A=@n15.A, B=@n17), B=QualitySwitch(Default=@n20, Low=@n19, Medium=@n19))
  )
  n21 = TextureCoordinate()
  n22 = ParticleRandom()
  n23 = MaterialFunctionCall(XY=LinearInterpolate(A=@n21, B=OneMinus(None=@n21), Alpha=Round(None=@n22)) {constA=1, constB=-1}) {materialFunction=&Swizzle}
  n24 = AppendVector(A=DynamicParameter() {defaultValue=<rgba 0 1 1 0>, paramNames [Swipe, Scale U, Color Transition, VariantPicker]}.Scale U, B=Constant() {r=1})
  n25 = DynamicParameter() {defaultValue=<rgba 0 1 1 0>, paramNames [Swipe, Scale U, Color Transition, VariantPicker]}
  n26 = Constant()
  n27 = ComponentMask(None=TextureCoordinate()) {g=true}
  n28 = Multiply(A=SmoothStep(Value=@n27) {constMax=0.3, constMin=-0.05}, B=SmoothStep(Value=OneMinus(None=@n27)) {constMax=0.2, constMin=-0.05})
  n29 = MaterialFunctionCall(
   UVs=Add(A=Multiply(A=@n23.YX, B=@n24), B=AppendVector(A=@n25.Swipe, B=@n26))
   Texture Scale=AppendVector(A=Constant() {r=1}, B=Saturate(None=@n28))
  ) {materialFunction=&ScaleUVsByCenter}
  n30 = ComponentMask(None=@n29.UVs) {g=true}
  NamedRerouteDeclaration(None=AppendVector(A=ComponentMask(None=@n29.UVs) {r=true}, B=Saturate(None=@n30))) {name=UndistortedUVs, nodeColor=<rgba 1 0.8672 0 1>}
 ]
}
