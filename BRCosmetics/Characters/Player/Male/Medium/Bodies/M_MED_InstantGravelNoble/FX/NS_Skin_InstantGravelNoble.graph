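# Niagara system for the InstantGravelNoble player skin, adding movement-reactive trails, speed lines and lightning to the character, with a front-end preview mode.
# LightTrail draws bone-following ribbons, SpeedLinesAndLightning spawns sprites around sampled bones, and SplineThickenMeshes plays lightning and swoosh meshes ordered and mirrored by user arrays.
# System movement-state flags and timelines steer spawning, size and motion.
# A cyan-variant flag switches color curves, while time-of-day scales brightness.
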NiagaraSystem /BRCosmetics/Characters/Player/Male/Medium/Bodies/M_MED_InstantGravelNoble/FX/NS_Skin_InstantGravelNoble
dependencies {
 usedDataInterfaces [
  &NiagaraDataInterfaceArrayFloat3
  &NiagaraDataInterfaceColorCurve
  &NiagaraDataInterfaceCurve
  &NiagaraDataInterfaceSkeletalMesh
 ]
 usedDynamicInputs [
  &Boolean_AndOperation
  &VectorLength
  &MakeFloatFromVector
  &FNTimeOfDayScalar
 ]
 usedModules [
  &self:SystemSpawnScript.SystemScriptSource.SystemScriptGraph.NiagaraNodeAssignment_0.SetVariables_#1
  &self:SystemSpawnScript.SystemScriptSource.SystemScriptGraph.NiagaraNodeAssignment_5.SetVariables_#2
  &SystemState
  &self:SystemSpawnScript.SystemScriptSource.SystemScriptGraph.NiagaraNodeAssignment_6.SetVariables_#3
  &self:SystemSpawnScript.SystemScriptSource.SystemScriptGraph.NiagaraNodeAssignment_7.SetVariables_#4
  &Timeline
  &self:SystemSpawnScript.SystemScriptSource.SystemScriptGraph.NiagaraNodeAssignment_1.SetVariables_#5
  &self:SystemSpawnScript.SystemScriptSource.SystemScriptGraph.NiagaraNodeAssignment_2.SetVariables_#6
 ]
}
emitters {
 LightTrail {
  emitterSpawnScript [/SetVariables(Emitter.ColorVariant:NiagaraFloat=&MakeCustomFloatFromBool(Bool=$System.IsCyanVariant, True Float=0, False Float=1)) {isSetParameters=true}]
  emitterUpdateScript [
   &EmitterState(
    Inactive Response=::NewEnumerator1
    Life Cycle Mode=::NewEnumerator1
    Loop Behavior=::NewEnumerator1
    Loop Duration Mode=::NewEnumerator1
    UseLoopDelay=false
   )
   &FN_SpawnAnimTrail(AnimTrailWidthMode=::FromSecond, Enable:NiagaraBool=&Boolean_NotOperation(Bool=$User.IsSkydiving))
   &SampleSkeletalMesh(
    AdditionalYawPitchRoll=false
    Bone Sampling Mode:ENiagaraBoneSamplingMode=::NewEnumerator4
    Skeletal Mesh:NiagaraDataInterfaceSkeletalMesh=$System.SkeletalMesh
   )
  ]
  particleSpawnScript [
   &FN_InitializeAnimTrailBoneIndecies()
   &FN_InitializeAnimTrail(
    AnimTrailWidth=2.5
    Color=<rgba 1 1 1 0.2>
    Lifetime=0.333
    Position Offset=&MakeCustomVectorFromBool(Bool=$User.bIsFrontEndPreview, True Vector=<0 0 2>, False Vector=<0 0 0>)
    SkeletalMesh=$System.SkeletalMesh
   )
   &AddVelocity(
    Rotation Coordinate Space=::World
    Velocity=&Multiply_Vector(
     A=<1 1 0>
     B=&RotateVectorByQuaternion(Vector To Rotate=<0 -1 0>, Quaternion=$Emitter.SampleSkeletalMesh.SampledRotation)
    )
    Velocity Speed Scale=&MakeCustomFloatFromBool(Bool=$User.bIsFrontEndPreview, True Float=666, False Float=0)
   )
  ]
  particleUpdateScript [
   &DynamicMaterialParameters(
    Index 0 Param 1=&MakeCustomFloatFromBool(Bool=$System.IsCyanVariant, True Float=0, False Float=1)
    Index 0 Param 2=6
    Index 0 Param 3=12
    Index 0 Param 4=6
    Module.Index0Param0Write=true
    Module.Index0Param1Write=true
    Module.Index0Param2Write=true
    Module.Index0Param3Write=true
   )
   &ParticleState()
   &FN_UpdateAnimTrail(SkeletalMesh=$System.SkeletalMesh)
   /SetVariables(Particles.OffsetPosition:NiagaraPosition=&MakeCustomVectorFromBool(
     Bool=&Boolean_OrOperation(Bool1=&Boolean_OrOperation(Bool1=$System.IsSprinting, Bool2=$User.bIsFrontEndPreview), Bool2=$User.IsSkydiving)
     True Vector=Linked() {linkedVariable {name=Particles.Position, type=NiagaraPosition}}
     False Vector=&MakeCustomVectorFromBool(
      Bool=$System.IsRunning
      True Vector=&AddVectorToPosition(Position=$Particles.Position, Vector=<0 0 2>)
      False Vector=Linked() {linkedVariable {name=Particles.Position, type=NiagaraPosition}}
     )
    )) {isSetParameters=true}
   /SetVariables(
    Particles.DynamicParametersSides:Vector4f=&MakeVector4(X=&MakeCustomFloatFromBool(Bool=$System.IsCyanVariant, True Float=0, False Float=1), Y=16, Z=16, W=16)
    Particles.PositionSideA:NiagaraPosition=&AddVectorToPosition(
     Position=$Particles.OffsetPosition
     Vector=&Multiply_VectorByFloat(Vector=$Engine.Owner.SystemXAxis, Float=-0.5)
    )
    Particles.PositionSideB:NiagaraPosition=&AddVectorToPosition(
     Position=$Particles.OffsetPosition
     Vector=&Multiply_VectorByFloat(Vector=$Engine.Owner.SystemXAxis, Float=0.5)
    )
   ) {isSetParameters=true}
   &ScaleColor(
    Scale Alpha=&MakeCustomFloatFromBool(Bool=$User.bIsFrontEndPreview, True Float=$System.Timeline002.Time, False Float=1)
    Scale RGB=&VectorFromFloat(Value=$System.TimeOfDayBrightnessMultiplier)
    ScaleA=true
    ScaleRGB=true
   )
   &SolveForcesAndVelocity(Clamp Velocity=false, Limit Acceleration=false)
  ]
  rendererClasses [&NiagaraRibbonRendererProperties]
  renderers [
   {rendererClass=&NiagaraRibbonRendererProperties}
   {rendererClass=&NiagaraRibbonRendererProperties, rendererIndex=1}
   {rendererClass=&NiagaraRibbonRendererProperties, rendererIndex=2}
  ]
 }
 SpeedLinesAndLightning {
  emitterUpdateScript [
   &EmitterState(
    Inactive Response=::NewEnumerator1
    Life Cycle Mode=::NewEnumerator1
    Loop Behavior=::NewEnumerator1
    Loop Duration Mode=::NewEnumerator1
    UseLoopDelay=false
   )
   &SpawnRate(
    Spawn Probability=&MakeCustomFloatFromBool(
     Bool=$System.FrontEndActive
     True Float=1
     False Float=&MakeCustomFloatFromBool(
      Bool=$User.IsSkydiving
      True Float=&MakeCustomFloatFromBool(Bool=$System.IsGlidingForward, True Float=1, False Float=0)
      False Float=&MakeCustomFloatFromBool(
       Bool=$System.IsRunning
       True Float=&MakeCustomFloatFromBool(Bool=$System.IsSprinting, True Float=1, False Float=0.666)
       False Float=0
      )
     )
    )
    SpawnRate=16
    Use Spawn Probability=true
   )
   /SetVariables(Emitter.SpriteLengthMultiplier:NiagaraFloat=&MakeCustomFloatFromBool(
     Bool=$System.IsGliding
     True Float=2.5
     False Float=&MakeCustomFloatFromBool(Bool=$System.IsSprinting, True Float=1.5, False Float=1)
    )) {isSetParameters=true}
  ]
  particleSpawnScript [
   &InitializeParticle(
    Color Mode=::NewEnumerator0
    Lifetime=0.333
    Position Mode=::NewEnumerator0
    Sprite Size=<8 32>
    Sprite Size Mode=::NewEnumerator4
    Write Material Random=true
   )
   &SkeletalMeshLocation(
    Bone Sampling Mode=::NewEnumerator4
    MaskSpawnGroup=false
    Orientation Sampling=::NewEnumerator1
    Sampled Position Offset=<10 -10 0>
    Skeletal Mesh=$System.SkeletalMesh
   )
   &ShapeLocation(
    Additional Quaternion Rotation=true
    Additional Yaw / Pitch / Roll=<0.0864 0 0.25>
    MaskSpawnGroup=false
    Override Seed=false
    Ring Radius=16
    Rotation Mode=::NewEnumerator3
    Rotation Quaternion=$Particles.SkeletalMeshLocation.SampledRotation
    Shape Primitive=::NewEnumerator4
    UseOwnerScale=false
   )
   &AddVelocity(
    Rotation Coordinate Space=::Simulation
    Velocity=&MakeCustomVectorFromBool(
     Bool=$System.IsGliding
     True Vector=&TransformVector(Vector=<0 1 0>, Destination Space=::Simulation)
     False Vector=&Multiply_Vector(A=<1 1 0>, B=$Particles.ShapeLocation.ShapeNormal)
    )
    Velocity Speed Scale=&MakeCustomFloatFromBool(
     Bool=$System.FrontEndActive
     True Float=-166
     False Float=&MakeCustomFloatFromBool(Bool=$System.IsGliding, True Float=-333, False Float=0)
    )
   )
   &ApplyInitialForces()
   /SetVariables(Particles.SpriteAlignment=&Multiply_Vector(A=<-1 -1 0>, B=$Particles.ShapeLocation.ShapeNormal)) {isSetParameters=true}
   /SetVariables(
    Particles.LightningSize:Vector2f=&MakeVector2D(
     X=&MakeCustomFloatFromBool(Bool=$System.IsGliding, True Float=12, False Float=7)
     Y=&Multiply_Float(A=&RandomRangeFloat(Minimum=9, Maximum=30), B=$Emitter.SpriteLengthMultiplier)
    )
    Particles.SpeedlineDynamicParameters:Vector4f=&MakeVector4(X=4, Y=66, Z=0.2, W=0)
    Particles.SpeedlineSize:Vector2f=&MakeVector2D(
     X=8
     Y=&Multiply_Float(A=&RandomRangeFloat(Minimum=20, Maximum=30), B=$Emitter.SpriteLengthMultiplier)
    )
    Particles.VisibilityTag=&MakeIntFromFloat(Float=&UniformAOrBFloat(
      A=0
      Distribution Weight=&MakeCustomFloatFromBool(
       Bool=$System.FrontEndActive
       True Float=0.4
       False Float=&MakeCustomFloatFromBool(
        Bool=$System.IsGliding
        True Float=0.6
        False Float=&MakeCustomFloatFromBool(Bool=$System.IsSprinting, True Float=1, False Float=0.1)
       )
      )
     ))
   ) {isSetParameters=true}
  ]
  particleUpdateScript [
   /SetVariables(
    Particles.OpacityMultiplier:NiagaraFloat=&FloatFromCurve(
     FloatCurve=DataInterface() {propertyValues {Curve {keys [<at 0 0>, {tangentMode=RCTM_User, time=0.1, value=1}, {tangentMode=RCTM_User, time=1, value=0}]}}}
     Scale Curve=&FN_FrontEndFloatModifier(Gameplay=1, Front End=$System.Timeline002.Time)
    )
    Particles.SpriteAlignment=&MakeCustomVectorFromBool(
     Bool=$User.bIsFrontEndPreview
     True Vector=$Particles.Initial.SpriteAlignment
     False Vector=&MakeCustomVectorFromBool(
      Bool=$System.IsGliding
      True Vector=&Lerp_Vector(
       A=&NormalizeVector(Vector To Normalize=$Engine.Owner.Velocity, Safe Normalize=true, Safe Fallback Threshold=0, Safe Fallback Vector=<0 0 0.001>)
       B=&TransformVector(Vector=<0 1 0>, Destination Space=::World)
       Alpha=&FloatFromCurve(FloatCurve=DataInterface() {propertyValues {Curve {keys [
            {interpMode=RCIM_Cubic, tangentMode=RCTM_Break, time=0, value=0.8}
            {arriveTangent=-0.7673, interpMode=RCIM_Cubic, tangentMode=RCTM_Break, time=1, value=0.5}
           ]}}})
      )
      False Vector=&NormalizeVector(
       Vector To Normalize=&Subtract_Position(A=$Particles.Initial.Position, B=$Particles.Position)
       Safe Normalize=true
       Safe Fallback Threshold=0
       Safe Fallback Vector=&Multiply_VectorByFloat(
        Vector=&NormalizeVector(Vector To Normalize=$Engine.Owner.Velocity, Safe Normalize=true, Safe Fallback Threshold=0, Safe Fallback Vector=<0 0 1>)
        Float=1
       )
      )
     )
    )
   ) {isSetParameters=true}
   /SetVariables(
    Particles.GlowColor=&MultiplyLinearColor(
     Linear Color=&MakeCustomLinearColorFromBool(
      Bool=$System.IsCyanVariant
      True Color=&ColorFromCurve(ColorCurve=DataInterface() {propertyValues {
         AlphaCurve {keys [<at 0 1>]}
         BlueCurve {keys [<at 0 1 RCIM_Cubic>,<at 0.333 1>,<at 1 1 RCIM_Cubic>]}
         GreenCurve {keys [<at 0 0.6667 RCIM_Cubic>,<at 0.333 0.1>,<at 1 0 RCIM_Cubic>]}
         RedCurve {keys [<at 0 0 RCIM_Cubic>,<at 0.333 0.1>,<at 1 0.25 RCIM_Cubic>]}
        }})
      False Color=&ColorFromCurve(ColorCurve=DataInterface() {propertyValues {
         AlphaCurve {keys [<at 0 1>]}
         BlueCurve {keys [<at 0 0 RCIM_Cubic>,<at 0.333 0>,<at 1 0 RCIM_Cubic>]}
         GreenCurve {keys [<at 0 0.4167 RCIM_Cubic>,<at 0.333 0.175>,<at 1 0.0667 RCIM_Cubic>]}
         RedCurve {keys [<at 0 0.5 RCIM_Cubic>,<at 0.333 0.3>,<at 1 0.2 RCIM_Cubic>]}
        }})
     )
     Scale Factor=&MakeFloat4FromVectorAndFloat(
      "Vector (RGB)"=&VectorFromFloat(Value=$System.TimeOfDayBrightnessMultiplier)
      "Float (Alpha)"=$Particles.OpacityMultiplier
     )
    )
    Particles.GlowDynamicParameters:Vector4f=&MakeVector4(X=1, Y=33, Z=&FN_FrontEndFloatModifier(Gameplay=0.066, Front End=0.2), W=0)
    Particles.GlowSize=&Multiply_Vector2D(A=$Particles.LightningSize, B=<3.333 1.666>)
    Particles.LightningColor:LinearColor=&MakeLinearColorFromVectorAndFloat(
     "Vector (RGB)"=&VectorFromFloat(Value=$System.TimeOfDayBrightnessMultiplier)
     "Float (Alpha)"=$Particles.OpacityMultiplier
    )
    Particles.LightningDynamicParameters:Vector4f=&MakeVector4(
     X=&Add_Float(
      A=&FloatFromCurve(FloatCurve=DataInterface() {propertyValues {Curve {keys [<at 0 0>,<at 1 1>]}}}, Scale Curve=0.3)
      B=&RandomRangeFloat(Minimum=0, Maximum=1)
     )
     Y=&Divide_Float(
      A=&MakeFloatFromVector2d(Vector 2D=$Particles.Initial.LightningSize, Channel=::NewEnumerator1)
      B=&MakeCustomFloatFromBool(Bool=$System.IsGliding, True Float=128, False Float=64)
     )
     Z=&FloatFromCurve(FloatCurve=DataInterface() {propertyValues {Curve {keys [<at 0 0>,<at 0.7 1>]}}})
     W=$User.VariantPicker
    )
    Particles.LightningSize:Vector2f=&Multiply_Vector2D_ByFloat(
     Float=&FloatFromCurve(FloatCurve=DataInterface() {propertyValues {Curve {keys [{tangentMode=RCTM_User, time=0, value=0}, {tangentMode=RCTM_User, time=0.1, value=1}]}}})
     Vector2D=$Particles.Initial.LightningSize
    )
    Particles.SpeedlineColor:LinearColor=&MultiplyLinearColor(
     Linear Color=&MakeCustomLinearColorFromBool(
      Bool=$System.IsCyanVariant
      True Color=&ColorFromCurve(ColorCurve=DataInterface() {propertyValues {
         AlphaCurve {keys [<at 0 1>]}
         BlueCurve {keys [<at 0 0.8333 RCIM_Cubic>,<at 0.333 1>,<at 1 1 RCIM_Cubic>]}
         GreenCurve {keys [<at 0 1 RCIM_Cubic>,<at 0.333 0.6667>,<at 1 0 RCIM_Cubic>]}
         RedCurve {keys [<at 0 0 RCIM_Cubic>,<at 0.333 0>,<at 1 1 RCIM_Cubic>]}
        }})
      False Color=&ColorFromCurve(ColorCurve=DataInterface() {propertyValues {
         AlphaCurve {keys [<at 0 1>]}
         BlueCurve {keys [<at 0 0 RCIM_Cubic>,<at 0.333 0>,<at 1 0 RCIM_Cubic>]}
         GreenCurve {keys [<at 0 0.625 RCIM_Cubic>,<at 0.333 0.4667>,<at 1 0.1667 RCIM_Cubic>]}
         RedCurve {keys [<at 0 0.75 RCIM_Cubic>,<at 0.333 0.8>,<at 1 0.5 RCIM_Cubic>]}
        }})
     )
     Scale Factor=&MakeFloat4FromVectorAndFloat(
      "Vector (RGB)"=&VectorFromFloat(Value=$System.TimeOfDayBrightnessMultiplier)
      "Float (Alpha)"=$Particles.OpacityMultiplier
     )
    )
    Particles.SpeedlineSize:Vector2f=&Multiply_Vector2D_ByFloat(
     Float=&FloatFromCurve(FloatCurve=DataInterface() {propertyValues {Curve {keys [<at 0 0>,<at 0.1 1>,<at 1 1>]}}})
     Vector2D=$Particles.Initial.SpeedlineSize
    )
   ) {isSetParameters=true}
   &InheritSourceMovement(
    Applied Position Scale=&MakeCustomVectorFromBool(
     Bool=$System.IsGliding
     True Vector=&Multiply_VectorByFloat(
      Vector=<1 1 1>
      Float=&FloatFromCurve(
       FloatCurve=DataInterface() {propertyValues {Curve {keys [
           {interpMode=RCIM_Cubic, tangentMode=RCTM_Break, time=0, value=1}
           {arriveTangent=-0.7233, interpMode=RCIM_Cubic, tangentMode=RCTM_Break, time=1, value=0.666}
          ]}}}
       Scale Curve=0.95
      )
     )
     False Vector=&MakeCustomVectorFromBool(Bool=$System.IsGrinding, True Vector=<0.8 0.8 0.8>, False Vector=<0.5 0.5 0.5>)
    )
    ApplyForce=false
    ApplyPosition=true
    ApplyPositionFalloff=false
   )
   &ParticleState()
   &SolveForcesAndVelocity(Clamp Velocity=false, Limit Acceleration=false)
  ]
  rendererClasses [&NiagaraSpriteRendererProperties]
  renderers [
   {rendererClass=&NiagaraSpriteRendererProperties}
   {rendererClass=&NiagaraSpriteRendererProperties, rendererIndex=1}
   {rendererClass=&NiagaraSpriteRendererProperties, rendererIndex=2}
  ]
 }
 SplineThickenMeshes {
  emitterUpdateScript [
   &EmitterState(
    Inactive Response=::NewEnumerator1
    Life Cycle Mode=::NewEnumerator1
    Loop Behavior=::NewEnumerator1
    Loop Duration Mode=::NewEnumerator1
    UseLoopDelay=false
   )
   &SpawnRate(
    Spawn Probability=&MakeCustomFloatFromBool(
     Bool=$System.FrontEndActive
     True Float=$System.Timeline002.Time
     False Float=&MakeCustomFloatFromBool(
      Bool=$System.IsFreefalling
      True Float=0
      False Float=&MakeCustomFloatFromBool(Bool=$System.IsSprinting, True Float=1, False Float=0)
     )
    )
    SpawnRate=&MakeCustomFloatFromBool(Bool=$System.FrontEndActive, True Float=9, False Float=&Lerp_Float(A=55, B=11, Alpha=$System.Timeline001.Time))
    Use Spawn Probability=true
   )
  ]
  particleSpawnScript [
   &InitializeParticle(
    Color=&MakeLinearColorFromVectorAndFloat("Vector (RGB)"=&VectorFromFloat(Value=$System.TimeOfDayBrightnessMultiplier), "Float (Alpha)"=1)
    Lifetime=&MakeCustomFloatFromBool(
     Bool=$User.bIsFrontEndPreview
     True Float=&RandomRangeFloat(Minimum=0.466, Maximum=0.466)
     False Float=&Lerp_Float(A=0.166, B=0.466, Alpha=$System.Timeline001.Time)
    )
    Mesh Scale=&SelectVectorFromArray(
     Direct Array Index=&SelectIntFromArray(
      Direct Array Index=&ReturnParticleIDAttributes()
      Direct Array Mode=::NewEnumerator1
      Int Selection Array=$User.LightningMirroring
     )
     Direct Array Mode=::NewEnumerator1
     Vector Selection Array=DataInterface() {propertyValues {FloatData [<1 1 1>, <1 1 -1>]}}
    )
    Mesh Scale Mode=::NewEnumerator4
    UsePositionOffset=false
    Write Material Random=true
   )
   /SetVariables(
    Particles.HardGlowColor:LinearColor=&MultiplyLinearColor(
     Linear Color=$Particles.Color
     Scale Factor=&MakeFloat4FromVectorAndFloat("Vector (RGB)"=<0.5 0.5 0.5>, "Float (Alpha)"=0.1)
    )
    Particles.IsSoloSwoosh:NiagaraBool=&SetBoolByFloatComparison(A=0, B=&UniformAOrBFloat(A=-1, Distribution Weight=0.666))
    Particles.MeshIndex=&SelectIntFromArray(
     Direct Array Index=&ReturnParticleIDAttributes()
     Direct Array Mode=::NewEnumerator1
     Int Selection Array=$User.LightningOrder
    )
    Particles.VisibilityTag=&MakeIntFromFloat(Float=&MakeCustomFloatFromBool(
      Bool=$System.IsGliding
      True Float=0
      False Float=&UniformAOrBFloat(A=0, Distribution Weight=$System.Timeline001.Time)
     ))
   ) {isSetParameters=true}
   /SetVariables(Particles.LightningVisibilityTag:NiagaraInt32=&MakeCustomIntFromBool(Bool=$Particles.IsSoloSwoosh, True INT=-1, False INT=$Particles.VisibilityTag)) {isSetParameters=true}
  ]
  particleUpdateScript [
   &ParticleState()
   &SkeletalMeshLocation(Bone Sampling Mode=::NewEnumerator4, MaskSpawnGroup=false, Skeletal Mesh=$System.SkeletalMesh)
   /SetVariables(Particles.SwooshColor:LinearColor=&MultiplyLinearColor(
     Linear Color=&MakeCustomLinearColorFromBool(
      Bool=$System.IsCyanVariant
      True Color=&ColorFromCurve(ColorCurve=DataInterface() {propertyValues {
         AlphaCurve {keys [<at 0 1>]}
         BlueCurve {keys [<at 0 0.8333 RCIM_Cubic>,<at 0.333 1>,<at 1 1 RCIM_Cubic>]}
         GreenCurve {keys [<at 0 1 RCIM_Cubic>,<at 0.333 0.6667>,<at 1 0 RCIM_Cubic>]}
         RedCurve {keys [<at 0 0 RCIM_Cubic>,<at 0.333 0>,<at 1 1 RCIM_Cubic>]}
        }})
      False Color=&ColorFromCurve(ColorCurve=DataInterface() {propertyValues {
         AlphaCurve {keys [<at 0 1>]}
         BlueCurve {keys [<at 0 0 RCIM_Cubic>,<at 0.333 0>,<at 1 0 RCIM_Cubic>]}
         GreenCurve {keys [<at 0 0.625 RCIM_Cubic>,<at 0.333 0.4667>,<at 1 0.1667 RCIM_Cubic>]}
         RedCurve {keys [<at 0 0.75 RCIM_Cubic>,<at 0.333 0.8>,<at 1 0.5 RCIM_Cubic>]}
        }})
     )
     Scale Factor=&MakeFloat4FromVectorAndFloat(
      "Vector (RGB)"=&Multiply_VectorByFloat(Vector=<1 1 1>, Float=$System.TimeOfDayBrightnessMultiplier)
      "Float (Alpha)"=0.4
     )
    )) {isSetParameters=true}
   &DynamicMaterialParameters(
    Index 0 Param 1=&FloatFromCurve(FloatCurve=DataInterface() {propertyValues {Curve {keys [{tangentMode=RCTM_User, time=0, value=0}, <at 0.333 0.5>, {tangentMode=RCTM_User, time=1, value=1}]}}})
    Index 0 Param 2=&FloatFromCurve(FloatCurve=DataInterface() {propertyValues {Curve {keys [<at 0.333 1>,<at 0.666 0.333>]}}})
    Index 0 Param 3=0.75
    Index 0 Param 4=&FloatFromCurve(
     FloatCurve=DataInterface() {propertyValues {Curve {keys [
         {interpMode=RCIM_Cubic, tangentMode=RCTM_User, time=0, value=0}
         {interpMode=RCIM_Cubic, tangentMode=RCTM_User, time=0.5, value=1}
         {tangentMode=RCTM_User, time=1, value=0}
        ]}}}
     Scale Curve=33
    )
    Index 1 Param 1=$User.VariantPicker
    Module.Index0Param0Write=true
    Module.Index0Param1Write=true
    Module.Index0Param2Write=true
    Module.Index0Param3Write=true
    Module.Index1Param0Write=true
    Module.Index1Param1Write=false
    Module.Index1Param2Write=false
    Module.Index1Param3Write=false
    Write Parameter Index 1=true
   )
   /SetVariables(Particles.MeshIndex=$Particles.PartitionParticles.Partition) {enabled=false, isSetParameters=true}
  ]
  rendererClasses [&NiagaraMeshRendererProperties]
  renderers [
   {rendererClass=&NiagaraMeshRendererProperties, rendererIndex=0}
   {rendererClass=&NiagaraMeshRendererProperties, rendererIndex=1}
   {rendererClass=&NiagaraMeshRendererProperties, rendererIndex=2}
   {rendererClass=&NiagaraMeshRendererProperties, rendererIndex=3}
   {rendererClass=&NiagaraMeshRendererProperties, rendererIndex=4}
   {rendererClass=&NiagaraMeshRendererProperties, rendererIndex=5}
   {rendererClass=&NiagaraMeshRendererProperties, rendererIndex=6}
   {rendererClass=&NiagaraMeshRendererProperties, rendererIndex=7}
  ]
 }
}
summary {
 systemName=NS_Skin_InstantGravelNoble
 emitters [
  {emitterName=SplineThickenMeshes, rendererClasses [&NiagaraMeshRendererProperties]}
  {emitterName=SpeedLinesAndLightning, rendererClasses [&NiagaraSpriteRendererProperties]}
  {emitterName=LightTrail, rendererClasses [&NiagaraRibbonRendererProperties]}
 ]
 userVariables [
  User.bIsFrontEnd {type=NiagaraBool}
  User.bIsFrontEndPreview {type=NiagaraBool}
  User.bisInVehicle {type=NiagaraBool}
  User.FrontEndReactivity {type=NiagaraBool}
  User.IsSkydiving {type=NiagaraBool}
  User.LightningMirroring {
   type=NiagaraDataInterfaceArrayInt32
   description=""
   defaultValue {
    struct=NiagaraExt_VariableValue_DataInterface
    value {
     dataInterface=&self:NiagaraDataInterfaceArrayInt32_1
     dataInterfaceClass=&NiagaraDataInterfaceArrayInt32
    }
   }
  }
  User.LightningOrder {
   type=NiagaraDataInterfaceArrayInt32
   description=""
   defaultValue {
    struct=NiagaraExt_VariableValue_DataInterface
    value {
     dataInterface=&self:NiagaraDataInterfaceArrayInt32_0
     dataInterfaceClass=&NiagaraDataInterfaceArrayInt32
    }
   }
  }
  User.RibbonIDOffset {type=NiagaraInt32}
  User.VariantPicker {type=NiagaraFloat, defaultValue=1, description=""}
  User.VelocityThresholdFreefalling {type=NiagaraFloat, defaultValue=750, description=""}
  User.VelocityThresholdGlidingForward {type=NiagaraFloat, defaultValue=1000, description=""}
  User.VelocityThresholdGrinding {type=NiagaraFloat, defaultValue=850, description=""}
  User.VelocityThresholdRunning {type=NiagaraFloat, defaultValue=400, description=""}
  User.VelocityThresholdSprinting {type=NiagaraFloat, defaultValue=600, description=""}
 ]
}
