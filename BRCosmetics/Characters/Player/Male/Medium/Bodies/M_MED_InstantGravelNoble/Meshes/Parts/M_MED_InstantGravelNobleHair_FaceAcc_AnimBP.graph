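# Post-copy Anim Blueprint that adds secondary hair motion to the InstantGravelNoble male character's hair/face accessory mesh.
# The parent mesh's pose passes through rotation constraints on mostly CST helper bones, then spring-driven AnimDynamics: back hair relative to spine_05, tighter-clamped front hair relative to pelvis.
# bIsJumping then blends in additive rotations on front strands.
# An unconnected near-duplicate back-hair dynamics chain also sits in the graph.
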AnimBlueprint /BRCosmetics/Characters/Player/Male/Medium/Bodies/M_MED_InstantGravelNoble/Meshes/Parts/M_MED_InstantGravelNobleHair_FaceAcc_AnimBP
target_skeleton=&/M_MED_InstantGravelNoble_FaceAcc_Skeleton
graphs [
 AnimGraph {nodes [
   n1 = Anim Dynamics - Bone: dyn_hair_back_6(Alpha=1, ExternalForce="0,0,0") {
    angularDampingOverride=0.9999
    angularSpringConstant=4000
    bAngularSpring=true
    bOverrideAngularDamping=true
    bUseGravityOverride=true
    bUseSphericalLimits=true
    gravityScale=0.5
    linearDampingOverride=0.9999
    simulationSpace=RootRelative
    alphaBoolBlend {blendOption=Linear}
    boundBone {boneName=dyn_hair_back_6}
    componentAppliedLinearAccClamp=<2000 2000 1200>
    componentLinearAccScale=<0.6 0.6 0.6>
    componentLinearVelScale=<0.7 0.7 0.7>
    gravityOverride {z=-1100}
    physicsBodyDefinitions [{
      collisionType=CustomSphere
      sphereCollisionRadius=1
      boundBone {boneName=dyn_hair_back_6}
      boxExtents=<10 15 15>
      constraintSetup {angularLimitsMax=<0 0 10>, angularLimitsMin=<0 0 -50>, angularTarget=<0 0 0>}
      localJointOffset=<0.6326 -0.8932 -0.0324>
     }]
    relativeSpaceBone {boneName=spine_05}
    sphericalLimits [{limitRadius=6, drivingBone {boneName=dyn_hair_back_1}, sphereLocalOffset=<4.7797 -1.9971 3.3879>}]
   }
   n2 = Anim Dynamics - Bone: dyn_hair_back_1(ComponentPose=@n1, Alpha=1, ExternalForce="0,0,0") {
    angularDampingOverride=0.9999
    angularSpringConstant=4000
    bAngularSpring=true
    bOverrideAngularDamping=true
    bUseGravityOverride=true
    bUseSphericalLimits=true
    gravityScale=0.8
    linearDampingOverride=0.9999
    simulationSpace=RootRelative
    alphaBoolBlend {blendOption=Linear}
    boundBone {boneName=dyn_hair_back_1}
    componentAppliedLinearAccClamp=<2000 2000 1200>
    componentLinearAccScale=<0.6 0.5 0.6>
    componentLinearVelScale=<0.7 0.6 0.7>
    gravityOverride {z=-1100}
    physicsBodyDefinitions [{
      sphereCollisionRadius=1
      boundBone {boneName=dyn_hair_back_1}
      boxExtents=<10 15 15>
      constraintSetup {angularLimitsMax=<0 5 3>, angularLimitsMin=<0 -35 -50>, angularTarget=<0 0 0>}
      localJointOffset=<3.0178 -0.503 0.9608>
     }]
    relativeSpaceBone {boneName=spine_05}
   }
   n3 = Anim Dynamics - Bone: dyn_hair_back_1_end(ComponentPose=@n2, Alpha=1, ExternalForce="0,0,0") {
    angularDampingOverride=0.9999
    angularSpringConstant=4000
    bAngularSpring=true
    bOverrideAngularDamping=true
    bUseGravityOverride=true
    gravityScale=0.8
    linearDampingOverride=0.9999
    simulationSpace=RootRelative
    alphaBoolBlend {blendOption=Linear}
    boundBone {boneName=dyn_hair_back_1_end}
    componentAppliedLinearAccClamp=<2000 2000 1000>
    componentLinearAccScale=<0.6 0.6 0.6>
    componentLinearVelScale=<0.7 0.7 0.7>
    gravityOverride {z=-1100}
    physicsBodyDefinitions [{
      sphereCollisionRadius=1
      boundBone {boneName=dyn_hair_back_1_end}
      boxExtents=<8 8 8>
      constraintSetup {angularLimitsMax=<0 10 5>, angularLimitsMin=<0 -10 -50>, angularTarget=<0 0 0>}
      localJointOffset=<2 0 0>
     }]
    relativeSpaceBone {boneName=spine_05}
   }
   n4 = Anim Dynamics - Bone: dyn_hair_back_2(ComponentPose=@n3, Alpha=1, ExternalForce="0,0,0") {
    angularDampingOverride=0.9999
    angularSpringConstant=4000
    bAngularSpring=true
    bOverrideAngularDamping=true
    bUseGravityOverride=true
    gravityScale=0.6
    linearDampingOverride=0.9999
    simulationSpace=RootRelative
    alphaBoolBlend {blendOption=Linear}
    boundBone {boneName=dyn_hair_back_2}
    componentAppliedLinearAccClamp=<2000 2000 1200>
    componentLinearAccScale=<0.6 0.6 0.6>
    componentLinearVelScale=<0.7 0.7 0.7>
    gravityOverride {z=-1100}
    physicsBodyDefinitions [{
      boundBone {boneName=dyn_hair_back_2}
      boxExtents=<10 15 15>
      constraintSetup {angularLimitsMax=<0 5 10>, angularLimitsMin=<0 -20 -50>, angularTarget=<0 0 0>}
      localJointOffset=<2 0 0>
     }]
    relativeSpaceBone {boneName=spine_05}
   }
   n5 = Anim Dynamics - Bone: dyn_hair_back_2_end(ComponentPose=@n4, Alpha=1, ExternalForce="0,0,0") {
    angularDampingOverride=0.9999
    angularSpringConstant=4000
    bAngularSpring=true
    bOverrideAngularDamping=true
    bUseGravityOverride=true
    bUseSphericalLimits=true
    gravityScale=0.6
    linearDampingOverride=0.9999
    simulationSpace=RootRelative
    alphaBoolBlend {blendOption=Linear}
    boundBone {boneName=dyn_hair_back_2_end}
    componentAppliedLinearAccClamp=<2000 2000 1000>
    componentLinearAccScale=<0.6 0.6 0.6>
    componentLinearVelScale=<0.7 0.7 0.7>
    gravityOverride {z=-1100}
    physicsBodyDefinitions [{
      collisionType=CustomSphere
      sphereCollisionRadius=1
      boundBone {boneName=dyn_hair_back_2_end}
      boxExtents=<8 8 8>
      constraintSetup {angularLimitsMax=<0 10 10>, angularLimitsMin=<0 -10 -50>, angularTarget=<0 0 0>}
      localJointOffset=<4.6251 0.0253 -2.3825>
     }]
    relativeSpaceBone {boneName=spine_05}
    sphericalLimits [{limitRadius=1, drivingBone {boneName=dyn_hair_back_1_end}, sphereLocalOffset=<4.082 2.5685 0.027>}]
   }
   n6 = Anim Dynamics - Bone: dyn_hair_back_3(ComponentPose=@n5, Alpha=1, ExternalForce="0,0,0") {
    angularDampingOverride=0.9999
    angularSpringConstant=4000
    bAngularSpring=true
    bOverrideAngularDamping=true
    bUseGravityOverride=true
    gravityScale=0.6
    linearDampingOverride=0.9999
    simulationSpace=RootRelative
    alphaBoolBlend {blendOption=Linear}
    boundBone {boneName=dyn_hair_back_3}
    componentAppliedLinearAccClamp=<2000 2000 1200>
    componentLinearAccScale=<0.6 0.6 0.6>
    componentLinearVelScale=<0.7 0.7 0.7>
    gravityOverride {z=-1100}
    physicsBodyDefinitions [{
      boundBone {boneName=dyn_hair_back_3}
      boxExtents=<10 15 15>
      constraintSetup {angularLimitsMax=<0 10 0>, angularLimitsMin=<0 -20 -50>, angularTarget=<0 0 0>}
      localJointOffset=<2 0 0>
     }]
    relativeSpaceBone {boneName=spine_05}
   }
   n7 = Anim Dynamics - Bone: dyn_hair_back_4(ComponentPose=@n6, Alpha=1, ExternalForce="0,0,0") {
    angularDampingOverride=0.9999
    angularSpringConstant=4000
    bAngularSpring=true
    bOverrideAngularDamping=true
    bUseGravityOverride=true
    bUsePlanarLimit=false
    gravityScale=0.7
    linearDampingOverride=0.9999
    simulationSpace=RootRelative
    alphaBoolBlend {blendOption=Linear}
    boundBone {boneName=dyn_hair_back_4}
    componentAppliedLinearAccClamp=<2000 2000 1200>
    componentLinearAccScale=<0.6 0.6 0.6>
    componentLinearVelScale=<0.7 0.7 0.7>
    gravityOverride {z=-1100}
    physicsBodyDefinitions [{
      sphereCollisionRadius=1
      boundBone {boneName=dyn_hair_back_4}
      boxExtents=<10 15 15>
      constraintSetup {angularLimitsMax=<0 25 15>, angularLimitsMin=<0 -20 -50>, angularTarget=<0 0 0>}
      localJointOffset=<2 0 0>
     }]
    relativeSpaceBone {boneName=spine_05}
   }
   Anim Dynamics - Bone: dyn_hair_back_5(ComponentPose=@n7, Alpha=1, ExternalForce="0,0,0") {
    angularDampingOverride=0.9999
    angularSpringConstant=4000
    bAngularSpring=true
    bOverrideAngularDamping=true
    bUseGravityOverride=true
    bUsePlanarLimit=false
    bUseSphericalLimits=true
    gravityScale=0.4
    linearDampingOverride=0.9999
    simulationSpace=RootRelative
    alphaBoolBlend {blendOption=Linear}
    boundBone {boneName=dyn_hair_back_5}
    componentAppliedLinearAccClamp=<2000 2000 1200>
    componentLinearAccScale=<0.6 0.6 0.6>
    componentLinearVelScale=<0.7 0.7 0.7>
    gravityOverride {z=-1100}
    physicsBodyDefinitions [{
      collisionType=CustomSphere
      sphereCollisionRadius=1
      boundBone {boneName=dyn_hair_back_5}
      boxExtents=<15 10 15>
      constraintSetup {angularLimitsMax=<0 40 30>, angularLimitsMin=<0 -15 -50>, angularTarget=<0 0 0>}
      localJointOffset=<2 0 0>
     }]
    relativeSpaceBone {boneName=spine_05}
    sphericalLimits [{limitRadius=3, drivingBone {boneName=dyn_hair_back_4}, sphereLocalOffset=<-0.0073 2.3171 -0.498>}]
   }
   n8 = "Constraint - dyn_hair_back_4_CST (RR)"!(
    ComponentPose=Local To Component(LocalPose=Copy Pose From Mesh() {bUseAttachedParent=true})
    Alpha=1
    ConstraintWeights_0=0.5
    ConstraintWeights_1:float=0.5
   ) -> (Pose) {
    alphaBoolBlend {blendOption=Linear}
    boneToModify {boneName=dyn_hair_back_4_CST}
    constraintSetup [{transformType=Rotation, targetBone {boneName=dyn_hair_back_2}}, {transformType=Rotation, targetBone {boneName=dyn_hair_back_3}}]
    constraintWeights [1,1]
   }
   n9 = "Constraint - dyn_hair_back_5_CST (RR)"!(ComponentPose=@n8, Alpha=1, ConstraintWeights_0=0.5, ConstraintWeights_1:float=0.5) -> (Pose) {
    alphaBoolBlend {blendOption=Linear}
    boneToModify {boneName=dyn_hair_back_5_CST}
    constraintSetup [{transformType=Rotation, targetBone {boneName=dyn_hair_back_2}}, {transformType=Rotation, targetBone {boneName=dyn_hair_back_4}}]
    constraintWeights [1,1]
   }
   n10 = "Constraint - dyn_hair_back_2_CST (RR)"!(ComponentPose=@n9, Alpha=1, ConstraintWeights_0=1, ConstraintWeights_1:float=0) -> (Pose) {
    alphaBoolBlend {blendOption=Linear}
    boneToModify {boneName=dyn_hair_back_2_CST}
    constraintSetup [{transformType=Rotation, targetBone {boneName=dyn_hair_back_3}}, {transformType=Rotation, targetBone {boneName=dyn_hair_back_6}}]
    constraintWeights [1,1]
   }
   n11 = "Constraint - dyn_hair_back_1_CST (RR)"!(ComponentPose=@n10, Alpha=1, ConstraintWeights_0=0.5, ConstraintWeights_1:float=0.5) -> (Pose) {
    alphaBoolBlend {blendOption=Linear}
    boneToModify {boneName=dyn_hair_back_1_CST}
    constraintSetup [{transformType=Rotation, targetBone {boneName=dyn_hair_back_2}}, {transformType=Rotation, targetBone {boneName=dyn_hair_back_3}}]
    constraintWeights [1,1]
   }
   n12 = "Constraint - dyn_hair_back_6_CST (RR)"!(ComponentPose=@n11, Alpha=1, ConstraintWeights_0=0.2, ConstraintWeights_1:float=0.8) -> (Pose) {
    alphaBoolBlend {blendOption=Linear}
    boneToModify {boneName=dyn_hair_back_6_CST}
    constraintSetup [{transformType=Rotation, targetBone {boneName=dyn_hair_back_5}}, {transformType=Rotation, targetBone {boneName=dyn_hair_back_2}}]
    constraintWeights [1,1]
   }
   n13 = "Constraint - dyn_hair_front_7_CST (RR)"!(ComponentPose=@n12, Alpha=1, ConstraintWeights_0=1, ConstraintWeights_1:float=0) -> (Pose) {
    alphaBoolBlend {blendOption=Linear}
    boneToModify {boneName=dyn_hair_front_7_CST}
    constraintSetup [{transformType=Rotation, targetBone {boneName=dyn_hair_front_10}}, {transformType=Rotation, targetBone {boneName=dyn_hair_front_12}}]
    constraintWeights [1,1]
   }
   n14 = "Constraint - dyn_hair_front_8_CST (RR)"!(ComponentPose=@n13, Alpha=1, ConstraintWeights_0=0, ConstraintWeights_1:float=1) -> (Pose) {
    alphaBoolBlend {blendOption=Linear}
    boneToModify {boneName=dyn_hair_front_8_CST}
    constraintSetup [{transformType=Rotation, targetBone {boneName=dyn_hair_front_9}}, {transformType=Rotation, targetBone {boneName=dyn_hair_front_10}}]
    constraintWeights [1,1]
   }
   n15 = "Constraint - dyn_hair_front_9 (RR)"!(ComponentPose=@n14, Alpha=1, ConstraintWeights_0=0.5, ConstraintWeights_1:float=0.5) -> (Pose) {
    alphaBoolBlend {blendOption=Linear}
    boneToModify {boneName=dyn_hair_front_9}
    constraintSetup [{transformType=Rotation, targetBone {boneName=dyn_hair_front_5_end}}, {transformType=Rotation, targetBone {boneName=dyn_hair_front_8}}]
    constraintWeights [1,1]
   }
   n16 = "Constraint - dyn_hair_front_5_CST (RR)"!(ComponentPose=@n15, Alpha=1, ConstraintWeights_0=0, ConstraintWeights_1:float=1) -> (Pose) {
    alphaBoolBlend {blendOption=Linear}
    boneToModify {boneName=dyn_hair_front_5_CST}
    constraintSetup [{transformType=Rotation, targetBone {boneName=dyn_hair_front_4}}, {transformType=Rotation, targetBone {boneName=dyn_hair_front_9}}]
    constraintWeights [1,1]
   }
   n17 = "Constraint - dyn_hair_front_4_CST (RR)"!(ComponentPose=@n16, Alpha=1, ConstraintWeights_0=0.5, ConstraintWeights_1:float=0.5) -> (Pose) {
    alphaBoolBlend {blendOption=Linear}
    boneToModify {boneName=dyn_hair_front_4_CST}
    constraintSetup [{transformType=Rotation, targetBone {boneName=dyn_hair_front_5}}, {transformType=Rotation, targetBone {boneName=dyn_hair_front_11}}]
    constraintWeights [1,1]
   }
   n18 = "Constraint - dyn_hair_front_12_CST (RR)"!(ComponentPose=@n17, Alpha=1, ConstraintWeights_0=0.5, ConstraintWeights_1:float=0.5) -> (Pose) {
    alphaBoolBlend {blendOption=Linear}
    boneToModify {boneName=dyn_hair_front_12_CST}
    constraintSetup [{transformType=Rotation, targetBone {boneName=dyn_hair_front_6}}, {transformType=Rotation, targetBone {boneName=dyn_hair_front_7}}]
    constraintWeights [1,1]
   }
   n19 = "Constraint - dyn_hair_front_6_CST (RR)"!(ComponentPose=@n18, Alpha=1, ConstraintWeights_0=0.5, ConstraintWeights_1:float=0.5) -> (Pose) {
    alphaBoolBlend {blendOption=Linear}
    boneToModify {boneName=dyn_hair_front_6_CST}
    constraintSetup [{transformType=Rotation, targetBone {boneName=dyn_hair_front_12}}, {transformType=Rotation, targetBone {boneName=dyn_hair_front_11}}]
    constraintWeights [1,1]
   }
   n20 = "Constraint - dyn_hair_front_11_CST (RR)"!(ComponentPose=@n19, Alpha=1, ConstraintWeights_0=0.5, ConstraintWeights_1:float=0.5) -> (Pose) {
    alphaBoolBlend {blendOption=Linear}
    boneToModify {boneName=dyn_hair_front_11_CST}
    constraintSetup [{transformType=Rotation, targetBone {boneName=dyn_hair_front_1}}, {transformType=Rotation, targetBone {boneName=dyn_hair_front_2}}]
    constraintWeights [1,1]
   }
   n21 = Anim Dynamics - Bone: dyn_hair_back_6(ComponentPose=@n20, Alpha=1, ExternalForce="0,0,0") {
    angularDampingOverride=0.9999
    angularSpringConstant=4000
    bAngularSpring=true
    bOverrideAngularDamping=true
    bUseGravityOverride=true
    bUseSphericalLimits=true
    gravityScale=0.5
    linearDampingOverride=0.9999
    simulationSpace=RootRelative
    alphaBoolBlend {blendOption=Linear}
    boundBone {boneName=dyn_hair_back_6}
    componentAppliedLinearAccClamp=<2000 2000 1200>
    componentLinearAccScale=<0.6 0.6 0.6>
    componentLinearVelScale=<0.7 0.7 0.7>
    gravityOverride {z=-1100}
    physicsBodyDefinitions [{
      sphereCollisionRadius=1
      boundBone {boneName=dyn_hair_back_6}
      boxExtents=<10 15 15>
      constraintSetup {angularLimitsMax=<0 15 10>, angularLimitsMin=<0 -5 -30>, angularTarget=<0 0 0>}
      localJointOffset=<0.6326 -0.8932 0.4>
     }]
    relativeSpaceBone {boneName=spine_05}
   }
   n22 = Anim Dynamics - Bone: dyn_hair_back_1(ComponentPose=@n21, Alpha=1, ExternalForce="0,0,0") {
    angularDampingOverride=0.9999
    angularSpringConstant=4000
    bAngularSpring=true
    bOverrideAngularDamping=true
    bUseGravityOverride=true
    bUseSphericalLimits=true
    gravityScale=0.8
    linearDampingOverride=0.9999
    simulationSpace=RootRelative
    alphaBoolBlend {blendOption=Linear}
    boundBone {boneName=dyn_hair_back_1}
    componentAppliedLinearAccClamp=<2000 2000 1200>
    componentLinearAccScale=<0.6 0.5 0.6>
    componentLinearVelScale=<0.7 0.6 0.7>
    gravityOverride {z=-1100}
    physicsBodyDefinitions [{
      sphereCollisionRadius=1
      boundBone {boneName=dyn_hair_back_1}
      boxExtents=<10 15 15>
      constraintSetup {angularLimitsMax=<0 5 3>, angularLimitsMin=<0 -35 -30>, angularTarget=<0 0 0>}
      localJointOffset=<3.0178 -0.503 0.9608>
     }]
    relativeSpaceBone {boneName=spine_05}
   }
   n23 = Anim Dynamics - Bone: dyn_hair_back_1_end(ComponentPose=@n22, Alpha=1, ExternalForce="0,0,0") {
    angularDampingOverride=0.9999
    angularSpringConstant=4000
    bAngularSpring=true
    bOverrideAngularDamping=true
    bUseGravityOverride=true
    gravityScale=0.8
    linearDampingOverride=0.9999
    simulationSpace=RootRelative
    alphaBoolBlend {blendOption=Linear}
    boundBone {boneName=dyn_hair_back_1_end}
    componentAppliedLinearAccClamp=<2000 2000 1000>
    componentLinearAccScale=<0.6 0.6 0.6>
    componentLinearVelScale=<0.7 0.7 0.7>
    gravityOverride {z=-1100}
    physicsBodyDefinitions [{
      sphereCollisionRadius=1
      boundBone {boneName=dyn_hair_back_1_end}
      boxExtents=<8 8 8>
      constraintSetup {angularLimitsMax=<0 10 5>, angularLimitsMin=<0 -10 -30>, angularTarget=<0 0 0>}
      localJointOffset=<2 0 0>
     }]
    relativeSpaceBone {boneName=spine_05}
   }
   n24 = Anim Dynamics - Bone: dyn_hair_back_2(ComponentPose=@n23, Alpha=1, ExternalForce="0,0,0") {
    angularDampingOverride=0.9999
    angularSpringConstant=4000
    bAngularSpring=true
    bOverrideAngularDamping=true
    bUseGravityOverride=true
    gravityScale=0.6
    linearDampingOverride=0.9999
    simulationSpace=RootRelative
    alphaBoolBlend {blendOption=Linear}
    boundBone {boneName=dyn_hair_back_2}
    componentAppliedLinearAccClamp=<2000 2000 1200>
    componentLinearAccScale=<0.6 0.6 0.6>
    componentLinearVelScale=<0.7 0.7 0.7>
    gravityOverride {z=-1100}
    physicsBodyDefinitions [{
      boundBone {boneName=dyn_hair_back_2}
      boxExtents=<10 15 15>
      constraintSetup {angularLimitsMax=<0 5 10>, angularLimitsMin=<0 -20 -30>, angularTarget=<0 0 0>}
      localJointOffset=<2 0 0>
     }]
    relativeSpaceBone {boneName=spine_05}
   }
   n25 = Anim Dynamics - Bone: dyn_hair_back_2_end(ComponentPose=@n24, Alpha=1, ExternalForce="0,0,0") {
    angularDampingOverride=0.9999
    angularSpringConstant=4000
    bAngularSpring=true
    bOverrideAngularDamping=true
    bUseGravityOverride=true
    bUseSphericalLimits=true
    gravityScale=0.6
    linearDampingOverride=0.9999
    simulationSpace=RootRelative
    alphaBoolBlend {blendOption=Linear}
    boundBone {boneName=dyn_hair_back_2_end}
    componentAppliedLinearAccClamp=<2000 2000 1000>
    componentLinearAccScale=<0.6 0.6 0.6>
    componentLinearVelScale=<0.7 0.7 0.7>
    gravityOverride {z=-1100}
    physicsBodyDefinitions [{
      collisionType=CustomSphere
      sphereCollisionRadius=1
      boundBone {boneName=dyn_hair_back_2_end}
      boxExtents=<8 8 8>
      constraintSetup {angularLimitsMax=<0 10 10>, angularLimitsMin=<0 -10 -30>, angularTarget=<0 0 0>}
      localJointOffset=<4.6251 0.0253 -2.3825>
     }]
    relativeSpaceBone {boneName=spine_05}
    sphericalLimits [{limitRadius=1, drivingBone {boneName=dyn_hair_back_1_end}, sphereLocalOffset=<4.082 2.5685 0.027>}]
   }
   n26 = Anim Dynamics - Bone: dyn_hair_back_3(ComponentPose=@n25, Alpha=1, ExternalForce="0,0,0") {
    angularDampingOverride=0.9999
    angularSpringConstant=4000
    bAngularSpring=true
    bOverrideAngularDamping=true
    bUseGravityOverride=true
    gravityScale=0.6
    linearDampingOverride=0.9999
    simulationSpace=RootRelative
    alphaBoolBlend {blendOption=Linear}
    boundBone {boneName=dyn_hair_back_3}
    componentAppliedLinearAccClamp=<2000 2000 1200>
    componentLinearAccScale=<0.6 0.6 0.6>
    componentLinearVelScale=<0.7 0.7 0.7>
    gravityOverride {z=-1100}
    physicsBodyDefinitions [{
      boundBone {boneName=dyn_hair_back_3}
      boxExtents=<10 15 15>
      constraintSetup {angularLimitsMax=<0 10 0>, angularLimitsMin=<0 -20 -30>, angularTarget=<0 0 0>}
      localJointOffset=<2 0 0>
     }]
    relativeSpaceBone {boneName=spine_05}
   }
   n27 = Anim Dynamics - Bone: dyn_hair_back_4(ComponentPose=@n26, Alpha=1, ExternalForce="0,0,0") {
    angularDampingOverride=0.9999
    angularSpringConstant=4000
    bAngularSpring=true
    bOverrideAngularDamping=true
    bUseGravityOverride=true
    bUsePlanarLimit=false
    gravityScale=0.7
    linearDampingOverride=0.9999
    simulationSpace=RootRelative
    alphaBoolBlend {blendOption=Linear}
    boundBone {boneName=dyn_hair_back_4}
    componentAppliedLinearAccClamp=<2000 2000 1200>
    componentLinearAccScale=<0.6 0.6 0.6>
    componentLinearVelScale=<0.7 0.7 0.7>
    gravityOverride {z=-1100}
    physicsBodyDefinitions [{
      sphereCollisionRadius=1
      boundBone {boneName=dyn_hair_back_4}
      boxExtents=<10 15 15>
      constraintSetup {angularLimitsMax=<0 25 15>, angularLimitsMin=<0 -20 -30>, angularTarget=<0 0 0>}
      localJointOffset=<2 0 0>
     }]
    relativeSpaceBone {boneName=spine_05}
   }
   n28 = Anim Dynamics - Bone: dyn_hair_back_5(ComponentPose=@n27, Alpha=1, ExternalForce="0,0,0") {
    angularDampingOverride=0.9999
    angularSpringConstant=4000
    bAngularSpring=true
    bOverrideAngularDamping=true
    bUseGravityOverride=true
    bUsePlanarLimit=false
    bUseSphericalLimits=true
    gravityScale=0.4
    linearDampingOverride=0.9999
    simulationSpace=RootRelative
    alphaBoolBlend {blendOption=Linear}
    boundBone {boneName=dyn_hair_back_5}
    componentAppliedLinearAccClamp=<2000 2000 1200>
    componentLinearAccScale=<0.6 0.6 0.6>
    componentLinearVelScale=<0.7 0.7 0.7>
    gravityOverride {z=-1100}
    physicsBodyDefinitions [{
      collisionType=CustomSphere
      sphereCollisionRadius=1
      boundBone {boneName=dyn_hair_back_5}
      boxExtents=<15 10 15>
      constraintSetup {angularLimitsMax=<0 40 30>, angularLimitsMin=<0 -15 -30>, angularTarget=<0 0 0>}
      localJointOffset=<2 0 0>
     }]
    relativeSpaceBone {boneName=spine_05}
    sphericalLimits [{limitRadius=3, drivingBone {boneName=dyn_hair_back_4}, sphereLocalOffset=<-0.0073 2.3171 -0.498>}]
   }
   n29 = Anim Dynamics - Bone: dyn_hair_front_1(ComponentPose=@n28, Alpha=1, ExternalForce="0,0,0") {
    angularDampingOverride=1
    angularSpringConstant=4000
    bAngularSpring=true
    bOverrideAngularDamping=true
    bUseGravityOverride=true
    gravityScale=0.4
    linearDampingOverride=0.9999
    simulationSpace=RootRelative
    alphaBoolBlend {blendOption=Linear}
    boundBone {boneName=dyn_hair_front_1}
    componentAppliedLinearAccClamp=<100 300 150>
    componentLinearAccScale=<0.1 0.1 0.5>
    componentLinearVelScale=<0.2 0.3 0.6>
    gravityOverride {z=-1100}
    physicsBodyDefinitions [{
      sphereCollisionRadius=1
      boundBone {boneName=dyn_hair_front_1}
      boxExtents=<14 14 14>
      constraintSetup {angularLimitsMax=<0 20 5>, angularLimitsMin=<0 -30 -50>, angularTarget=<0 0 0>}
      localJointOffset=<4 0 0>
     }]
    relativeSpaceBone {boneName=pelvis}
   }
   n30 = Anim Dynamics - Bone: dyn_hair_front_2(ComponentPose=@n29, Alpha=1, ExternalForce="0,0,0") {
    angularDampingOverride=1
    angularSpringConstant=4000
    bAngularSpring=true
    bOverrideAngularDamping=true
    bUseGravityOverride=true
    gravityScale=0.1
    linearDampingOverride=0.9999
    simulationSpace=RootRelative
    alphaBoolBlend {blendOption=Linear}
    boundBone {boneName=dyn_hair_front_2}
    componentAppliedLinearAccClamp=<100 300 100>
    componentLinearAccScale=<0.1 0.2 0.3>
    componentLinearVelScale=<0.2 0.4 0.4>
    gravityOverride {z=-1100}
    physicsBodyDefinitions [{
      sphereCollisionRadius=1
      boundBone {boneName=dyn_hair_front_2}
      boxExtents=<10 10 15>
      constraintSetup {angularLimitsMax=<10 10 20>, angularLimitsMin=<-10 -15 -50>, angularTarget=<0 0 0>}
      localJointOffset=<4 0 0>
     }]
    relativeSpaceBone {boneName=pelvis}
   }
   n31 = Anim Dynamics - Bone: dyn_hair_front_11(ComponentPose=@n30, Alpha=1, ExternalForce="0,0,0") {
    angularDampingOverride=1
    angularSpringConstant=4000
    bAngularSpring=true
    bOverrideAngularDamping=true
    bUseGravityOverride=true
    gravityScale=0.1
    linearDampingOverride=0.9999
    simulationSpace=RootRelative
    alphaBoolBlend {blendOption=Linear}
    boundBone {boneName=dyn_hair_front_11}
    componentAppliedLinearAccClamp=<100 300 100>
    componentLinearAccScale=<0.1 0.2 0.3>
    componentLinearVelScale=<0.2 0.4 0.4>
    gravityOverride {z=-1100}
    physicsBodyDefinitions [{
      sphereCollisionRadius=1
      boundBone {boneName=dyn_hair_front_11}
      boxExtents=<10 10 15>
      constraintSetup {angularLimitsMax=<0 0 3>, angularLimitsMin=<0 -50 -50>, angularTarget=<0 0 0>}
      localJointOffset=<4 0 0>
     }]
    relativeSpaceBone {boneName=pelvis}
   }
   n32 = Anim Dynamics - Bone: dyn_hair_front_3(ComponentPose=@n31, Alpha=1, ExternalForce="0,0,0") {
    angularDampingOverride=1
    angularSpringConstant=4000
    bAngularSpring=true
    bOverrideAngularDamping=true
    bUseGravityOverride=true
    gravityScale=0.1
    linearDampingOverride=0.9999
    simulationSpace=RootRelative
    alphaBoolBlend {blendOption=Linear}
    boundBone {boneName=dyn_hair_front_3}
    componentAppliedLinearAccClamp=<100 300 100>
    componentLinearAccScale=<0.1 0.2 0.3>
    componentLinearVelScale=<0.2 0.4 0.4>
    gravityOverride {z=-1100}
    physicsBodyDefinitions [{
      sphereCollisionRadius=1
      boundBone {boneName=dyn_hair_front_3}
      boxExtents=<10 10 15>
      constraintSetup {angularLimitsMax=<0 15 5>, angularLimitsMin=<0 -15 -50>, angularTarget=<0 0 0>}
      localJointOffset=<4 0 0>
     }]
    relativeSpaceBone {boneName=pelvis}
   }
   n33 = Anim Dynamics - Bone: dyn_hair_front_4(ComponentPose=@n32, Alpha=1, ExternalForce="0,0,0") {
    angularDampingOverride=0.9999
    angularSpringConstant=4000
    bAngularSpring=true
    bOverrideAngularDamping=true
    bUseGravityOverride=true
    gravityScale=0.4
    linearDampingOverride=0.9999
    simulationSpace=RootRelative
    alphaBoolBlend {blendOption=Linear}
    boundBone {boneName=dyn_hair_front_4}
    componentAppliedLinearAccClamp=<100 50 100>
    componentLinearAccScale=<0.1 0.1 0.3>
    componentLinearVelScale=<0.2 0.2 0.4>
    gravityOverride {z=-1100}
    physicsBodyDefinitions [{
      sphereCollisionRadius=1
      boundBone {boneName=dyn_hair_front_4}
      boxExtents=<14 14 22>
      constraintSetup {angularLimitsMax=<0 2 5>, angularLimitsMin=<0 -22 -15>, angularTarget=<0 0 0>}
      localJointOffset=<4 1 0>
     }]
    relativeSpaceBone {boneName=pelvis}
   }
   n34 = Anim Dynamics - Bone: dyn_hair_front_4_end(ComponentPose=@n33, Alpha=1, ExternalForce="0,0,0") {
    angularDampingOverride=0.9999
    angularSpringConstant=4000
    bAngularSpring=true
    bOverrideAngularDamping=true
    bUseGravityOverride=true
    gravityScale=0.4
    linearDampingOverride=0.9999
    simulationSpace=RootRelative
    alphaBoolBlend {blendOption=Linear}
    boundBone {boneName=dyn_hair_front_4_end}
    componentAppliedLinearAccClamp=<100 50 100>
    componentLinearAccScale=<0.1 0.1 0.3>
    componentLinearVelScale=<0.2 0.2 0.4>
    gravityOverride {z=-1100}
    physicsBodyDefinitions [{
      sphereCollisionRadius=1
      boundBone {boneName=dyn_hair_front_4_end}
      boxExtents=<10 10 15>
      constraintSetup {angularLimitsMax=<0 10 5>, angularLimitsMin=<0 -40 -15>, angularTarget=<0 0 0>}
      localJointOffset=<4 1 0>
     }]
    relativeSpaceBone {boneName=pelvis}
   }
   n35 = Anim Dynamics - Bone: dyn_hair_front_5(ComponentPose=@n34, Alpha=1, ExternalForce="0,0,0") {
    angularDampingOverride=0.9999
    angularSpringConstant=4000
    bAngularSpring=true
    bOverrideAngularDamping=true
    bUseGravityOverride=true
    bUseSphericalLimits=true
    gravityScale=0.4
    linearDampingOverride=0.9999
    simulationSpace=RootRelative
    alphaBoolBlend {blendOption=Linear}
    boundBone {boneName=dyn_hair_front_5}
    componentAppliedLinearAccClamp=<100 50 100>
    componentLinearAccScale=<0.1 0.1 0.3>
    componentLinearVelScale=<0.2 0.2 0.4>
    gravityOverride {z=-1100}
    physicsBodyDefinitions [{
      sphereCollisionRadius=1
      boundBone {boneName=dyn_hair_front_5}
      boxExtents=<10 10 15>
      constraintSetup {angularLimitsMax=<0 5 3>, angularLimitsMin=<0 -10 -15>, angularTarget=<0 0 0>}
      localJointOffset=<4 0 -1>
     }]
    relativeSpaceBone {boneName=pelvis}
   }
   n36 = Anim Dynamics - Bone: dyn_hair_front_5_end(ComponentPose=@n35, Alpha=1, ExternalForce="0,0,0") {
    angularDampingOverride=0.9999
    angularSpringConstant=4000
    bAngularSpring=true
    bOverrideAngularDamping=true
    bUseGravityOverride=true
    bUseSphericalLimits=true
    gravityScale=0.4
    linearDampingOverride=0.9999
    simulationSpace=RootRelative
    alphaBoolBlend {blendOption=Linear}
    boundBone {boneName=dyn_hair_front_5_end}
    componentAppliedLinearAccClamp=<100 50 100>
    componentLinearAccScale=<0.1 0.1 0.3>
    componentLinearVelScale=<0.2 0.2 0.4>
    gravityOverride {z=-1100}
    physicsBodyDefinitions [{
      sphereCollisionRadius=1
      boundBone {boneName=dyn_hair_front_5_end}
      boxExtents=<10 10 15>
      constraintSetup {angularLimitsMax=<0 -10 0>, angularLimitsMin=<0 -40 -15>, angularTarget=<0 0 0>}
      localJointOffset=<4 0 0.5>
     }]
    relativeSpaceBone {boneName=pelvis}
   }
   n37 = Anim Dynamics - Bone: dyn_hair_front_6(ComponentPose=@n36, Alpha=1, ExternalForce="0,0,0") {
    angularDampingOverride=0.9999
    angularSpringConstant=4000
    bAngularSpring=true
    bOverrideAngularDamping=true
    bUseGravityOverride=true
    gravityScale=0.4
    linearDampingOverride=0.9999
    simulationSpace=RootRelative
    alphaBoolBlend {blendOption=Linear}
    boundBone {boneName=dyn_hair_front_6}
    componentAppliedLinearAccClamp=<100 300 150>
    componentLinearAccScale=<0.1 0.1 0.5>
    componentLinearVelScale=<0.2 0.3 0.6>
    gravityOverride {z=-1100}
    physicsBodyDefinitions [{
      sphereCollisionRadius=1
      boundBone {boneName=dyn_hair_front_6}
      boxExtents=<12 12 18>
      constraintSetup {angularLimitsMax=<0 12 -3>, angularLimitsMin=<0 -20 -15>, angularTarget=<0 0 0>}
      localJointOffset=<4 1 -1.5>
     }]
    relativeSpaceBone {boneName=pelvis}
   }
   n38 = Anim Dynamics - Bone: dyn_hair_front_12(ComponentPose=@n37, Alpha=1, ExternalForce="0,0,0") {
    angularDampingOverride=0.9999
    angularSpringConstant=4000
    bAngularSpring=true
    bOverrideAngularDamping=true
    bUseGravityOverride=true
    gravityScale=0.3
    linearDampingOverride=0.9999
    simulationSpace=RootRelative
    alphaBoolBlend {blendOption=Linear}
    boundBone {boneName=dyn_hair_front_12}
    componentAppliedLinearAccClamp=<100 50 100>
    componentLinearAccScale=<0.1 0.1 0.3>
    componentLinearVelScale=<0.2 0.2 0.4>
    gravityOverride {z=-1100}
    physicsBodyDefinitions [{
      sphereCollisionRadius=1
      boundBone {boneName=dyn_hair_front_12}
      boxExtents=<16 16 23>
      constraintSetup {angularLimitsMax=<0 10 4>, angularLimitsMin=<0 -28 -15>, angularTarget=<0 0 0>}
      localJointOffset=<4 0 -0.4>
     }]
    relativeSpaceBone {boneName=pelvis}
   }
   n39 = Anim Dynamics - Bone: dyn_hair_front_12_end(ComponentPose=@n38, Alpha=1, ExternalForce="0,0,0") {
    angularDampingOverride=0.9999
    angularSpringConstant=4000
    bAngularSpring=true
    bOverrideAngularDamping=true
    bUseGravityOverride=true
    gravityScale=0.3
    linearDampingOverride=0.9999
    simulationSpace=RootRelative
    alphaBoolBlend {blendOption=Linear}
    boundBone {boneName=dyn_hair_front_12_end}
    componentAppliedLinearAccClamp=<100 50 100>
    componentLinearAccScale=<0.1 0.1 0.3>
    componentLinearVelScale=<0.2 0.2 0.4>
    gravityOverride {z=-1100}
    physicsBodyDefinitions [{
      sphereCollisionRadius=1
      boundBone {boneName=dyn_hair_front_12_end}
      boxExtents=<10 10 15>
      constraintSetup {angularLimitsMax=<0 10 4>, angularLimitsMin=<0 -35 -15>, angularTarget=<0 0 0>}
      localJointOffset=<4 0 0.4>
     }]
    relativeSpaceBone {boneName=pelvis}
   }
   n40 = Anim Dynamics - Bone: dyn_hair_front_7(ComponentPose=@n39, Alpha=1, ExternalForce="0,0,0") {
    angularDampingOverride=0.9999
    angularSpringConstant=4000
    bAngularSpring=true
    bOverrideAngularDamping=true
    bUseGravityOverride=true
    gravityScale=0.3
    linearDampingOverride=0.9999
    simulationSpace=RootRelative
    alphaBoolBlend {blendOption=Linear}
    boundBone {boneName=dyn_hair_front_7}
    componentAppliedLinearAccClamp=<100 50 100>
    componentLinearAccScale=<0.1 0.1 0.3>
    componentLinearVelScale=<0.2 0.2 0.4>
    gravityOverride {z=-1100}
    physicsBodyDefinitions [{
      sphereCollisionRadius=1
      boundBone {boneName=dyn_hair_front_7}
      boxExtents=<16 16 23>
      constraintSetup {angularLimitsMax=<0 10 0>, angularLimitsMin=<0 -15 -15>, angularTarget=<0 0 0>}
      localJointOffset=<4 0 0>
     }]
    relativeSpaceBone {boneName=pelvis}
   }
   n41 = Anim Dynamics - Bone: dyn_hair_front_7_end(ComponentPose=@n40, Alpha=1, ExternalForce="0,0,0") {
    angularDampingOverride=0.9999
    angularSpringConstant=4000
    bAngularSpring=true
    bOverrideAngularDamping=true
    bUseGravityOverride=true
    bUseSphericalLimits=true
    gravityScale=0.3
    linearDampingOverride=0.9999
    simulationSpace=RootRelative
    alphaBoolBlend {blendOption=Linear}
    boundBone {boneName=dyn_hair_front_7_end}
    componentAppliedLinearAccClamp=<100 50 100>
    componentLinearAccScale=<0.1 0.1 0.3>
    componentLinearVelScale=<0.2 0.2 0.4>
    gravityOverride {z=-1100}
    physicsBodyDefinitions [{
      sphereCollisionRadius=1
      boundBone {boneName=dyn_hair_front_7_end}
      boxExtents=<10 10 15>
      constraintSetup {angularLimitsMax=<0 10 4>, angularLimitsMin=<0 -10 -15>, angularTarget=<0 0 0>}
      localJointOffset=<4 0 0>
     }]
    relativeSpaceBone {boneName=pelvis}
   }
   n42 = Anim Dynamics - Bone: dyn_hair_front_10(ComponentPose=@n41, Alpha=1, ExternalForce="0,0,0") {
    angularDampingOverride=0.9999
    angularSpringConstant=4000
    bAngularSpring=true
    bOverrideAngularDamping=true
    bUseGravityOverride=true
    gravityScale=0.3
    linearDampingOverride=0.9999
    simulationSpace=RootRelative
    alphaBoolBlend {blendOption=Linear}
    boundBone {boneName=dyn_hair_front_10}
    componentAppliedLinearAccClamp=<100 50 100>
    componentLinearAccScale=<0.1 0.1 0.3>
    componentLinearVelScale=<0.2 0.2 0.4>
    gravityOverride {z=-1100}
    physicsBodyDefinitions [{
      sphereCollisionRadius=1
      boundBone {boneName=dyn_hair_front_10}
      boxExtents=<16 16 23>
      constraintSetup {angularLimitsMax=<0 10 5>, angularLimitsMin=<0 -10 -15>, angularTarget=<0 0 0>}
      localJointOffset=<4 0 0>
     }]
    relativeSpaceBone {boneName=pelvis}
   }
   n43 = Anim Dynamics - Bone: dyn_hair_front_10_end(ComponentPose=@n42, Alpha=1, ExternalForce="0,0,0") {
    angularDampingOverride=0.9999
    angularSpringConstant=4000
    bAngularSpring=true
    bOverrideAngularDamping=true
    bUseGravityOverride=true
    bUseSphericalLimits=true
    gravityScale=0.3
    linearDampingOverride=0.9999
    simulationSpace=RootRelative
    alphaBoolBlend {blendOption=Linear}
    boundBone {boneName=dyn_hair_front_10_end}
    componentAppliedLinearAccClamp=<100 50 100>
    componentLinearAccScale=<0.1 0.1 0.3>
    componentLinearVelScale=<0.2 0.2 0.4>
    gravityOverride {z=-1100}
    physicsBodyDefinitions [{
      sphereCollisionRadius=1
      boundBone {boneName=dyn_hair_front_10_end}
      boxExtents=<10 10 15>
      constraintSetup {angularLimitsMax=<0 10 8>, angularLimitsMin=<0 -15 -15>, angularTarget=<0 0 0>}
      localJointOffset=<4 0 0>
     }]
    relativeSpaceBone {boneName=pelvis}
   }
   n44 = Anim Dynamics - Bone: dyn_hair_front_8(ComponentPose=@n43, Alpha=1, ExternalForce="0,0,0") {
    angularDampingOverride=0.9999
    angularSpringConstant=4000
    bAngularSpring=true
    bOverrideAngularDamping=true
    bUseGravityOverride=true
    gravityScale=0.3
    linearDampingOverride=0.9999
    simulationSpace=RootRelative
    alphaBoolBlend {blendOption=Linear}
    boundBone {boneName=dyn_hair_front_8}
    componentAppliedLinearAccClamp=<100 50 100>
    componentLinearAccScale=<0.1 0.1 0.3>
    componentLinearVelScale=<0.2 0.2 0.4>
    gravityOverride {z=-1100}
    physicsBodyDefinitions [{
      collisionType=CustomSphere
      sphereCollisionRadius=1
      boundBone {boneName=dyn_hair_front_8}
      boxExtents=<16 16 23>
      constraintSetup {angularLimitsMax=<0 15 0>, angularLimitsMin=<0 -10 -15>, angularTarget=<0 0 0>}
      localJointOffset=<4 0 0.5>
     }]
    relativeSpaceBone {boneName=pelvis}
   }
   n45 = Anim Dynamics - Bone: dyn_hair_front_8_end(ComponentPose=@n44, Alpha=1, ExternalForce="0,0,0") {
    angularDampingOverride=0.9999
    angularSpringConstant=4000
    bAngularSpring=true
    bOverrideAngularDamping=true
    bUseGravityOverride=true
    bUseSphericalLimits=true
    gravityScale=0.3
    linearDampingOverride=0.9999
    simulationSpace=RootRelative
    alphaBoolBlend {blendOption=Linear}
    boundBone {boneName=dyn_hair_front_8_end}
    componentAppliedLinearAccClamp=<100 50 100>
    componentLinearAccScale=<0.1 0.1 0.3>
    componentLinearVelScale=<0.2 0.2 0.4>
    gravityOverride {z=-1100}
    physicsBodyDefinitions [{
      collisionType=CustomSphere
      sphereCollisionRadius=1
      boundBone {boneName=dyn_hair_front_8_end}
      boxExtents=<10 10 15>
      constraintSetup {angularLimitsMax=<0 12 20>, angularLimitsMin=<0 -20 -15>, angularTarget=<0 0 0>}
      localJointOffset=<4 0 -0.4>
     }]
    relativeSpaceBone {boneName=pelvis}
    sphericalLimits [{limitRadius=8, drivingBone {boneName=head}, sphereLocalOffset=<6.763 3.8687 0.0166>}]
   }
   n46 = Anim Dynamics - Bone: dyn_hair_front_9(ComponentPose=@n45, Alpha=1, ExternalForce="0,0,0") {
    angularDampingOverride=0.9999
    angularSpringConstant=4000
    bAngularSpring=true
    bOverrideAngularDamping=true
    bUseGravityOverride=true
    gravityScale=0.3
    linearDampingOverride=0.9999
    simulationSpace=RootRelative
    alphaBoolBlend {blendOption=Linear}
    boundBone {boneName=dyn_hair_front_9}
    componentAppliedLinearAccClamp=<100 50 100>
    componentLinearAccScale=<0.1 0.1 0.3>
    componentLinearVelScale=<0.2 0.2 0.4>
    gravityOverride {z=-1100}
    physicsBodyDefinitions [{
      sphereCollisionRadius=1
      boundBone {boneName=dyn_hair_front_9}
      boxExtents=<16 16 23>
      constraintSetup {angularLimitsMax=<0 22 5>, angularLimitsMin=<0 -10 -15>, angularTarget=<0 0 0>}
      localJointOffset=<4 0 0.3>
     }]
    relativeSpaceBone {boneName=pelvis}
   }
   n47 = Anim Dynamics - Bone: dyn_hair_front_9_end(ComponentPose=@n46, Alpha=1, ExternalForce="0,0,0") {
    angularDampingOverride=0.9999
    angularSpringConstant=4000
    bAngularSpring=true
    bOverrideAngularDamping=true
    bUseGravityOverride=true
    bUseSphericalLimits=true
    gravityScale=0.3
    linearDampingOverride=0.9999
    simulationSpace=RootRelative
    alphaBoolBlend {blendOption=Linear}
    boundBone {boneName=dyn_hair_front_9_end}
    componentAppliedLinearAccClamp=<100 50 100>
    componentLinearAccScale=<0.1 0.1 0.3>
    componentLinearVelScale=<0.2 0.2 0.4>
    gravityOverride {z=-1100}
    physicsBodyDefinitions [{
      sphereCollisionRadius=1
      boundBone {boneName=dyn_hair_front_9_end}
      boxExtents=<10 10 15>
      constraintSetup {angularLimitsMax=<0 -5 15>, angularLimitsMin=<0 -10 -15>, angularTarget=<0 0 0>}
      localJointOffset=<4 1 0>
     }]
    relativeSpaceBone {boneName=pelvis}
   }
   n48 = Get bIsJumping()
   n49 = "Transform (Modify) Bone - Bone: dyn_hair_front_1"!(ComponentPose=@n47, bAlphaBoolEnabled:bool=@n48, Translation="0,0,0", Rotation="0,-12,0", Scale="1,1,1") -> (Pose) {
    alphaInputType=Bool
    rotationMode=BMM_Additive
    rotationSpace=BCS_BoneSpace
    alphaBoolBlend {blendInTime=0.4, blendOption=Linear, blendOutTime=0.4}
    boneToModify {boneName=dyn_hair_front_1}
    rotation {yaw=-12}
   }
   n50 = "Transform (Modify) Bone - Bone: dyn_hair_front_2"!(ComponentPose=@n49, bAlphaBoolEnabled:bool=@n48, Translation="0,0,0", Rotation="0,-12,0", Scale="1,1,1") -> (Pose) {
    alphaInputType=Bool
    rotationMode=BMM_Additive
    rotationSpace=BCS_BoneSpace
    alphaBoolBlend {blendInTime=0.4, blendOption=Linear, blendOutTime=0.4}
    boneToModify {boneName=dyn_hair_front_2}
    rotation {yaw=-12}
   }
   n51 = "Transform (Modify) Bone - Bone: dyn_hair_front_11"!(ComponentPose=@n50, bAlphaBoolEnabled:bool=@n48, Translation="0,0,0", Rotation="0,-12,0", Scale="1,1,1") -> (Pose) {
    alphaInputType=Bool
    rotationMode=BMM_Additive
    rotationSpace=BCS_BoneSpace
    alphaBoolBlend {blendInTime=0.4, blendOption=Linear, blendOutTime=0.4}
    boneToModify {boneName=dyn_hair_front_11}
    rotation {yaw=-12}
   }
   n52 = "Transform (Modify) Bone - Bone: dyn_hair_front_3"!(ComponentPose=@n51, bAlphaBoolEnabled:bool=@n48, Translation="0,0,0", Rotation="0,-12,0", Scale="1,1,1") -> (Pose) {
    alphaInputType=Bool
    rotationMode=BMM_Additive
    rotationSpace=BCS_BoneSpace
    alphaBoolBlend {blendInTime=0.4, blendOption=Linear, blendOutTime=0.4}
    boneToModify {boneName=dyn_hair_front_3}
    rotation {yaw=-12}
   }
   n53 = "Transform (Modify) Bone - Bone: dyn_hair_front_4"!(ComponentPose=@n52, bAlphaBoolEnabled:bool=@n48, Translation="0,0,0", Rotation="0,-15,0", Scale="1,1,1") -> (Pose) {
    alphaInputType=Bool
    rotationMode=BMM_Additive
    rotationSpace=BCS_BoneSpace
    alphaBoolBlend {blendInTime=0.4, blendOption=Linear, blendOutTime=0.4}
    boneToModify {boneName=dyn_hair_front_4}
    rotation {yaw=-15}
   }
   n54 = "Transform (Modify) Bone - Bone: dyn_hair_front_5"!(ComponentPose=@n53, bAlphaBoolEnabled:bool=@n48, Translation="0,0,0", Rotation="0,-12,0", Scale="1,1,1") -> (Pose) {
    alphaInputType=Bool
    rotationMode=BMM_Additive
    rotationSpace=BCS_BoneSpace
    alphaBoolBlend {blendInTime=0.4, blendOption=Linear, blendOutTime=0.4}
    boneToModify {boneName=dyn_hair_front_5}
    rotation {yaw=-12}
   }
   n55 = "Transform (Modify) Bone - Bone: dyn_hair_front_6"!(ComponentPose=@n54, bAlphaBoolEnabled:bool=@n48, Translation="0,0,0", Rotation="0,-18,0", Scale="1,1,1") -> (Pose) {
    alphaInputType=Bool
    rotationMode=BMM_Additive
    rotationSpace=BCS_BoneSpace
    alphaBoolBlend {blendInTime=0.4, blendOption=Linear, blendOutTime=0.4}
    boneToModify {boneName=dyn_hair_front_6}
    rotation {yaw=-18}
   }
   n56 = "Transform (Modify) Bone - Bone: dyn_hair_front_12"!(ComponentPose=@n55, bAlphaBoolEnabled:bool=@n48, Translation="0,0,0", Rotation="0,-20,0", Scale="1,1,1") -> (Pose) {
    alphaInputType=Bool
    rotationMode=BMM_Additive
    rotationSpace=BCS_BoneSpace
    alphaBoolBlend {blendInTime=0.4, blendOption=Linear, blendOutTime=0.4}
    boneToModify {boneName=dyn_hair_front_12}
    rotation {yaw=-20}
   }
   n57 = "Transform (Modify) Bone - Bone: dyn_hair_front_7"!(ComponentPose=@n56, bAlphaBoolEnabled:bool=@n48, Translation="0,0,0", Rotation="0,-9,0", Scale="1,1,1") -> (Pose) {
    alphaInputType=Bool
    rotationMode=BMM_Additive
    rotationSpace=BCS_BoneSpace
    alphaBoolBlend {blendInTime=0.4, blendOption=Linear, blendOutTime=0.4}
    boneToModify {boneName=dyn_hair_front_7}
    rotation {yaw=-9}
   }
   n58 = "Transform (Modify) Bone - Bone: dyn_hair_front_10"!(ComponentPose=@n57, bAlphaBoolEnabled:bool=@n48, Translation="0,0,0", Rotation="0,-13,0", Scale="1,1,1") -> (Pose) {
    alphaInputType=Bool
    rotationMode=BMM_Additive
    rotationSpace=BCS_BoneSpace
    alphaBoolBlend {blendInTime=0.4, blendOption=Linear, blendOutTime=0.4}
    boneToModify {boneName=dyn_hair_front_10}
    rotation {yaw=-13}
   }
   n59 = "Transform (Modify) Bone - Bone: dyn_hair_front_8"!(ComponentPose=@n58, bAlphaBoolEnabled:bool=@n48, Translation="0,0,0", Rotation="0,-6,0", Scale="1,1,1") -> (Pose) {
    alphaInputType=Bool
    rotationMode=BMM_Additive
    rotationSpace=BCS_BoneSpace
    alphaBoolBlend {blendInTime=0.4, blendOption=Linear, blendOutTime=0.4}
    boneToModify {boneName=dyn_hair_front_8}
    rotation {yaw=-6}
   }
   n60 = "Transform (Modify) Bone - Bone: dyn_hair_front_9"!(ComponentPose=@n59, bAlphaBoolEnabled:bool=@n48, Translation="0,0,0", Rotation="0,-16,0", Scale="1,1,1") -> (Pose) {
    alphaInputType=Bool
    rotationMode=BMM_Additive
    rotationSpace=BCS_BoneSpace
    alphaBoolBlend {blendInTime=0.4, blendOption=Linear, blendOutTime=0.4}
    boneToModify {boneName=dyn_hair_front_9}
    rotation {yaw=-16}
   }
   Output Pose(Result=Component To Local(ComponentPose=@n60)) {layerGroup=None}
  ]}
 EventGraph {nodes [
   TryGetPawnOwner()
   Event BlueprintUpdateAnimation()
  ]}
]
variables []
